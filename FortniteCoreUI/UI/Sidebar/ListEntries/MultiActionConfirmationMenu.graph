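# Sidebar confirmation menu for bulk friend actions, built on FortMultiFriendActionMenu, with YES and NO buttons bound to the native parent.
# Opening or an action update runs UpdateText, which switches on the action enum and only sets text for accepting all friend requests.
# Opening also plays a slide, shear and fade-in animation.
# An animation-finished callback, bound on activation and unbound on deactivation, focuses the NO button as the default.
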WidgetBlueprint /FortniteCoreUI/UI/Sidebar/ListEntries/MultiActionConfirmationMenu
animations [Opened {
  binding_count=1
  display_label=""
  length_seconds=0.3
  object=&self:Opened
  sequence {bindings [Border_Confirmation {
     id=#1
     possessed_class=Border
     tracks [
      {
       type=MovieSceneFloatTrack
       display_name=Render Opacity
       sections [{
         type=MovieSceneFloatSection
         end_frame=6
         start_frame=0
         channels [None {type=MovieSceneScriptingFloatChannel, thinned_from=4, keys [[0,0],[3,1],[6,1]]}]
        }]
      }
      {
       type=MovieScene2DTransformTrack
       sections [{
         type=MovieScene2DTransformSection
         end_frame=6
         start_frame=0
         channels [
          Translation.X {type=MovieSceneScriptingFloatChannel, thinned_from=4, keys [[0,-256],[3,0],[6,0]]}
          Translation.Y {type=MovieSceneScriptingFloatChannel}
          Angle {type=MovieSceneScriptingFloatChannel}
          Scale.X {type=MovieSceneScriptingFloatChannel}
          Scale.Y {type=MovieSceneScriptingFloatChannel}
          Shear.X {type=MovieSceneScriptingFloatChannel, thinned_from=4, keys [[0,16],[3,0],[6,0]]}
          Shear.Y {type=MovieSceneScriptingFloatChannel}
         ]
        }]
      }
     ]
    }]}
 }]
graphs [
 EventGraph {nodes [
   n1 = OnOpenedAnimationFinished() {is=title K2Node_CustomEvent}
   > SetFocus(self=Get Button_Decline() -> (Button_Decline:Common Button Base Object Reference))
   Event OnActionUpdated() {is=title K2Node_Event}
   > UpdateText()
   Event On Activated()
   n2 = Get Opened() -> (Opened:Widget Animation Object Reference)
   > BindToAnimationFinished(Animation=@n2, Delegate=@n1)
   Event On Deactivated()
   > UnbindFromAnimationFinished!(execute, self:User Widget Object Reference, Animation:Widget Animation Object Reference=@n2, Delegate:Delegate=@n1) -> (then)
   Event OnOpened() {is=title K2Node_Event}
   > UpdateText()
   > PlayAnimation(InAnimation=Get Opened())
  ]}
 UpdateText {nodes [
   UpdateText() {is=title K2Node_FunctionEntry}
   > Switch on EMultiFriendAction!(execute, Selection:EMultiFriendAction Enum=Get Action() -> (Action:EMultiFriendAction Enum=AcceptFriendRequest)) -> (AcceptFriendRequest:Exec) {is=title K2Node_SwitchEnum}
   > SetText(
    self=Get Text_ActionText() -> (Text_ActionText:Common Rich Text Block Object Reference)
    InText=MakeLiteralText(Value="LOCTEXT(\"Accept All Friend Request\")")
   )
  ]}
]
variables []
widgets {
 info {inheritedWidgetCount=2, parentClass=/Script/FortniteUI.FortMultiFriendActionMenu, rootWidgetClass=&Overlay, widgetCount=10}
 widgets [
  Overlay_0 {widgetClassPath=&Overlay, widget=Overlay_0}
  SizeBox_0 {
   widgetClassPath=&SizeBox
   parent=Overlay_0
   widget=SizeBox_0
   properties {bOverride_WidthOverride=true, widthOverride=520}
   slot {class=OverlaySlot}
  }
  VB_Content {widgetClassPath=&VerticalBox, parent=SizeBox_0, widget=VB_Content, slot {class=SizeBoxSlot}}
  Border_Confirmation {
   widgetClassPath=&Border
   parent=VB_Content
   widget=Border_Confirmation
   properties {
    navigation=&self:WidgetTree.Border_Confirmation.WidgetNavigation_0
    verticalAlignment=VAlign_Center
    visibility=SelfHitTestInvisible
    background {
     bIsDynamicallyLoaded=false
     drawAs=Image
     imageType=NoImage
     mirroring=NoMirror
     resourceName=None
     resourceObject=/Game/UI/Sidebar/Materials/M_UI_sidebar_Button_Selection_Inst
     tiling=NoTile
     imageSize=<32 32>
     margin=<ltrb 0 0 0 0>
     outlineSettings {
      bUseBrushTransparency=false
      roundingType=HalfHeightRadius
      width=0
      color {colorUseRule=UseColor_Specified, specifiedColor=<rgba 0 0 0 0>}
      cornerRadii=<0 0 0 0>
     }
     tintColor {colorUseRule=UseColor_Specified, specifiedColor=<rgba 1 1 1 1>}
     uVRegion {bIsValid=false, max=<0 0>, min=<0 0>}
    }
    padding=<ltrb 0 8 0 8>
    renderTransformPivot=<0 0.5>
   }
   slot {class=VerticalBoxSlot}
  }
  VerticalBox_0 {
   widgetClassPath=&VerticalBox
   parent=Border_Confirmation
   widget=VerticalBox_0
   slot {class=BorderSlot, properties {verticalAlignment=VAlign_Center, padding=<ltrb 0 8 0 8>}}
  }
  SizeBox_1 {
   widgetClassPath=&SizeBox
   parent=VerticalBox_0
   widget=SizeBox_1
   slot {class=VerticalBoxSlot, properties {horizontalAlignment=HAlign_Center, verticalAlignment=VAlign_Top, padding=<ltrb 16 8 0 16>}}
  }
  Text_ActionText {
   widgetClassPath=&CommonRichTextBlock
   bIsVariable=true
   parent=SizeBox_1
   widget=Text_ActionText
   properties {
    bOverrideDefaultStyle=true
    bTintInlineIcon=true
    mobileTextBlockScale=1.2
    wrapTextAt=450
    defaultTextStyleOverride {
     overflowPolicy=Clip
     transformPolicy=None
     colorAndOpacity {colorUseRule=UseColor_Specified, specifiedColor=<rgba 0.0331 0.402 1 1>}
     font {
      bForceMonospaced=false
      bMaterialIsStencil=false
      fontMaterial=null
      fontObject=&BurbankSmall
      letterSpacing=0
      monospacedWidth=1
      size=18
      skewAmount=0
      typefaceFontName=Black
      outlineSettings {
       bApplyOutlineToDropShadows=false
       bMiteredCorners=false
       bSeparateFillAlpha=false
       outlineMaterial=null
       outlineSize=0
       outlineColor=<rgba 0 0 0 1>
      }
     }
     highlightColor {colorUseRule=UseColor_Specified, specifiedColor=<rgba 0 0 0 1>}
     highlightShape {
      bIsDynamicallyLoaded=false
      drawAs=Image
      imageType=NoImage
      mirroring=NoMirror
      resourceName=None
      resourceObject=null
      tiling=NoTile
      imageSize=<32 32>
      margin=<ltrb 0 0 0 0>
      outlineSettings {
       bUseBrushTransparency=false
       roundingType=HalfHeightRadius
       width=0
       color {colorUseRule=UseColor_Specified, specifiedColor=<rgba 0 0 0 0>}
       cornerRadii=<0 0 0 0>
      }
      tintColor {colorUseRule=UseColor_Specified, specifiedColor=<rgba 1 1 1 1>}
      uVRegion {bIsValid=false, max=<0 0>, min=<0 0>}
     }
     selectedBackgroundColor {colorUseRule=UseColor_Foreground, specifiedColor=<rgba 1 0 1 1>}
     shadowColorAndOpacity=<rgba 0 0 0 1>
     shadowOffset=<0 0>
     strikeBrush {
      bIsDynamicallyLoaded=false
      drawAs=Image
      imageType=NoImage
      mirroring=NoMirror
      resourceName=None
      resourceObject=null
      tiling=NoTile
      imageSize=<32 32>
      margin=<ltrb 0 0 0 0>
      outlineSettings {
       bUseBrushTransparency=false
       roundingType=HalfHeightRadius
       width=0
       color {colorUseRule=UseColor_Specified, specifiedColor=<rgba 0 0 0 0>}
       cornerRadii=<0 0 0 0>
      }
      tintColor {colorUseRule=UseColor_Specified, specifiedColor=<rgba 1 1 1 1>}
      uVRegion {bIsValid=false, max=<0 0>, min=<0 0>}
     }
     underlineBrush {
      bIsDynamicallyLoaded=false
      drawAs=Image
      imageType=NoImage
      mirroring=NoMirror
      resourceName=None
      resourceObject=null
      tiling=NoTile
      imageSize=<32 32>
      margin=<ltrb 0 0 0 0>
      outlineSettings {
       bUseBrushTransparency=false
       roundingType=HalfHeightRadius
       width=0
       color {colorUseRule=UseColor_Specified, specifiedColor=<rgba 0 0 0 0>}
       cornerRadii=<0 0 0 0>
      }
      tintColor {colorUseRule=UseColor_Specified, specifiedColor=<rgba 1 1 1 1>}
      uVRegion {bIsValid=false, max=<0 0>, min=<0 0>}
     }
    }
   }
   slot {class=SizeBoxSlot}
  }
  ConfirmationButtons {
   widgetClassPath=&HorizontalBox
   bIsVariable=true
   parent=VerticalBox_0
   widget=ConfirmationButtons
   properties {navigation=&self:WidgetTree.ConfirmationButtons.WidgetNavigation_0}
   slot {class=VerticalBoxSlot, properties {padding=<ltrb 32 0 0 0>}}
  }
  Button_Confirm {
   widgetClassPath=&WBP_UIKit_Button_Regular
   bInherited=true
   bIsVariable=true
   parent=ConfirmationButtons
   widget=Button_Confirm
   properties {text=YES}
   slot {class=HorizontalBoxSlot, properties {padding=<ltrb 0 0 8 0>, size {sizeRule=Fill, value=1}}}
  }
  Button_Decline {
   widgetClassPath=&WBP_UIKit_Button_Regular
   bInherited=true
   bIsVariable=true
   parent=ConfirmationButtons
   widget=Button_Decline
   properties {text=NO}
   slot {class=HorizontalBoxSlot, properties {padding=<ltrb 8 0 0 0>, size {sizeRule=Fill, value=1}}}
  }
 ]
}
