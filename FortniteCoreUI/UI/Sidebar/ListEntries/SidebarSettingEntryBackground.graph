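# Background widget for sidebar setting entries that wraps arbitrary content in a material border with hover feedback.
# Content goes in a named slot inside a size box with a minimum height.
# For non-touch input, mouse enter and leave play the OnHover animation forward and reverse, driving the border material's AnimateHighlight parameter.
# Both then broadcast OnMouseEnterChanged to listeners, and enter also plays a hover sound.
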WidgetBlueprint /FortniteCoreUI/UI/Sidebar/ListEntries/SidebarSettingEntryBackground
animations [OnHover {
  binding_count=1
  length_seconds=0.1
  object=&self:OnHover
  sequence {bindings [AnimBoundEntryBorder {
     id=#1
     possessed_class=Border
     tracks [{
       type=MovieSceneWidgetMaterialTrack
       sections [{type=MovieSceneParameterSection, channels [AnimateHighlight {type=MovieSceneScriptingFloatChannel, keys [[0,0],[1,1],[2,1]]}]}]
      }]
    }]}
 }]
graphs [
 EventGraph {nodes [
   Event OnMouseEnter()
   n1 = IsUsingTouch(PlayerController=GetOwningPlayer())
   > n2 = Branch(Condition=@n1)
   n3 = Get OnHover() -> (OnHover:Widget Animation Object Reference)
   PlayAnimationForward(execute=@n2.else, InAnimation=@n3)
   > Call On Mouse Enter Changed!(execute:Exec, self:Sidebar Setting Entry Background Object Reference, IsEnterEvent:bool=true) -> (then:Exec) {is=title K2Node_CallDelegate}
   > PlaySound2D(Sound=/Game/Sounds/Fort_UI_Sounds/FrontEnd/New_FrontEnd/Fortnite_UI_MainMenu_Tab_Hover_01_Cue)
   Event OnMouseLeave!() -> (OutputDelegate, then, MouseEvent:"Pointer Event Structure (by ref)") {is=title K2Node_Event}
   > n4 = Branch(Condition=@n1)
   PlayAnimationReverse(execute=@n4.else, InAnimation=@n3)
   > Call On Mouse Enter Changed()
  ]}
 OnMouseEnterChanged {nodes [OnMouseEnterChanged!() -> (then, IsEnterEvent:bool=~) {is=title K2Node_FunctionEntry}]}
]
variables [OnMouseEnterChanged {}]
widgets {
 info {inheritedWidgetCount=0, parentClass=&UserWidget, rootWidgetClass=&Border, widgetCount=3}
 widgets [
  AnimBoundEntryBorder {
   widgetClassPath=&Border
   widget=AnimBoundEntryBorder
   properties {
    background {
     bIsDynamicallyLoaded=false
     drawAs=Image
     imageType=NoImage
     mirroring=NoMirror
     resourceName=None
     resourceObject=/FortniteCoreUI/UI/Sidebar/Materials/M_UI_SidebarSettingsBorder
     tiling=NoTile
     imageSize=<32 32>
     margin=<ltrb 0 0 0 0>
     outlineSettings {
      bUseBrushTransparency=false
      roundingType=HalfHeightRadius
      width=0
      color {colorUseRule=UseColor_Specified, specifiedColor=<rgba 0 0 0 0>}
      cornerRadii=<0 0 0 0>
     }
     tintColor {colorUseRule=UseColor_Specified, specifiedColor=<rgba 1 1 1 1>}
     uVRegion {bIsValid=false, max=<0 0>, min=<0 0>}
    }
    padding=<ltrb 0 0 0 0>
   }
  }
  SizeBox_2 {
   widgetClassPath=&SizeBox
   parent=AnimBoundEntryBorder
   widget=SizeBox_2
   properties {bOverride_MinDesiredHeight=true, heightOverride=64, minDesiredHeight=50}
   slot {class=BorderSlot, properties {padding=<ltrb 0 0 0 0>}}
  }
  Content {
   widgetClassPath=&NamedSlot
   bIsVariable=true
   parent=SizeBox_2
   widget=Content
   properties {bExposeOnInstanceOnly=true}
   slot {class=SizeBoxSlot}
  }
 ]
}
