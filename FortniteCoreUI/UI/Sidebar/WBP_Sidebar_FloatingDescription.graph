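# Floating description tooltip for the sidebar, showing text in a UIKit info box that starts collapsed.
# PreConstruct restyles the box, swapping its backplate for a dark rounded-material background with a dynamic corner radius.
# GetTooltipVisibility requires the ClamOne runtime option, a show flag and nonzero show time, then starts the FortFloatingTooltipViewModel timer.
# AnimateTooltip plays fade-and-slide intro or outro animations, leaving debug PrintStrings in both branches.
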WidgetBlueprint /FortniteCoreUI/UI/Sidebar/WBP_Sidebar_FloatingDescription
animations [
 Anim_Intro {
  binding_count=2
  length_seconds=0.5
  object=&self:Anim_Intro
  sequence {bindings [
    SB_Content {
     id=#1
     possessed_class=SizeBox
     tracks [
      {
       type=MovieSceneFloatTrack
       display_name=Render Opacity
       sections [{type=MovieSceneFloatSection, end_frame=10, start_frame=0, channels [None {type=MovieSceneScriptingFloatChannel, keys [[0,0],[10,1]]}]}]
      }
      {
       type=MovieSceneByteTrack
       display_name=Visibility
       sections [{type=MovieSceneByteSection, end_frame=10, start_frame=0, channels [None {type=MovieSceneScriptingByteChannel, keys [[0,1],[1,4],[10,4]]}]}]
      }
     ]
    }
    GridPanel_0 {
     id=#2
     possessed_class=GridPanel
     tracks [{
       type=MovieScene2DTransformTrack
       sections [{
         type=MovieScene2DTransformSection
         end_frame=10
         start_frame=0
         channels [
          Translation.X {type=MovieSceneScriptingFloatChannel, keys [[0,100],[10,0]]}
          Translation.Y {type=MovieSceneScriptingFloatChannel, const [0,1]}
          Angle {type=MovieSceneScriptingFloatChannel, const [0,1]}
          Scale.X {type=MovieSceneScriptingFloatChannel, const [1,1]}
          Scale.Y {type=MovieSceneScriptingFloatChannel, const [1,1]}
          Shear.X {type=MovieSceneScriptingFloatChannel, const [0,1]}
          Shear.Y {type=MovieSceneScriptingFloatChannel, const [0,1]}
         ]
        }]
      }]
    }
   ]}
 }
 Anim_Outro {
  binding_count=2
  length_seconds=0.5
  object=&self:Anim_Outro
  sequence {bindings [
    SB_Content {
     id=#3
     possessed_class=SizeBox
     tracks [
      {
       type=MovieSceneFloatTrack
       display_name=Render Opacity
       sections [{type=MovieSceneFloatSection, end_frame=10, start_frame=0, channels [None {type=MovieSceneScriptingFloatChannel, keys [[1,1],[10,0]]}]}]
      }
      {
       type=MovieSceneByteTrack
       display_name=Visibility
       sections [{type=MovieSceneByteSection, end_frame=10, start_frame=0, channels [None {type=MovieSceneScriptingByteChannel, keys [[0,4],[10,1]]}]}]
      }
     ]
    }
    GridPanel_0 {
     id=#4
     possessed_class=GridPanel
     tracks [{
       type=MovieScene2DTransformTrack
       sections [{
         type=MovieScene2DTransformSection
         end_frame=10
         start_frame=0
         channels [
          Translation.X {type=MovieSceneScriptingFloatChannel, keys [[0,0],[10,100]]}
          Translation.Y {type=MovieSceneScriptingFloatChannel, const [0,1]}
          Angle {type=MovieSceneScriptingFloatChannel, const [0,1]}
          Scale.X {type=MovieSceneScriptingFloatChannel, const [1,1]}
          Scale.Y {type=MovieSceneScriptingFloatChannel, const [1,1]}
          Shear.X {type=MovieSceneScriptingFloatChannel, const [0,1]}
          Shear.Y {type=MovieSceneScriptingFloatChannel, const [0,1]}
         ]
        }]
      }]
    }
   ]}
 }
]
graphs [
 AnimateTooltip {nodes [
   n1 = AnimateTooltip!() -> (then, IsVisible:bool=~) {is=title K2Node_FunctionEntry}
   > n2 = Branch(Condition=@n1)
   > StopAnimation(InAnimation=Get Anim_Outro() -> (Anim_Outro:Widget Animation Object Reference))
   > PlayAnimation(InAnimation=Get Anim_Intro() -> (Anim_Intro:Widget Animation Object Reference))
   > PrintString!(execute, InString=Play Intro Animation, bPrintToScreen, bPrintToLog, TextColor, Duration, Key) -> (then)
   StopAnimation(execute=@n2.else, InAnimation=Get Anim_Intro())
   > PrintString(InString=Play Outro Animation)
   > PlayAnimation(InAnimation=Get Anim_Outro())
  ]}
 EventGraph {nodes [
   n1 = Event PreConstruct()
   > n2 = Sequence()
   > UpdateInfoBoxVisual()
   > SetVisibility(self=Get SB_Content() -> (SB_Content:Size Box Object Reference), InVisibility=Collapsed)
   Branch(execute=@n2.then_1, Condition=@n1.IsDesignTime)
   > SetDescriptionText!(execute, self, Text:Text="LOCTEXT(\"Only your friends will be able to see mutual friends on your profile\")") -> (then)
   Event Destruct()
   > PrintString(InString="Description Destructed!")
  ]}
 GetTooltipVisibility {nodes [
   n1 = GetTooltipVisibility!() -> (then, bShouldBeShown:bool=~, ShowTime:double=~, DescriptionText:Text) {is=title K2Node_FunctionEntry}
   > n2 = Branch(Condition=GetIsClamOneEnabled!(self:Fort Runtime Options Object Reference=GetRuntimeOptions()) -> (ReturnValue:bool))
   > n3 = Branch(Condition=AND Boolean(A=@n1, B="Not Equal (Float)"(A=@n1.ShowTime)))
   > StartTimer(self:Fort Floating Tooltip View Model Object Reference=Get FortFloatingTooltipViewModel() -> (FortFloatingTooltipViewModel:Fort Floating Tooltip View Model Object Reference))
   > Return Node!(execute, IsVisible:bool=true)
   Return Node(execute=[@n3.else, @n2.else], IsVisible=false)
  ]}
 SetDescriptionText {nodes [
   n1 = SetDescriptionText!() -> (then, Text:Text) {is=title K2Node_FunctionEntry}
   > SetText(self:WBP UIKit Info Box Object Reference=Get InfoBox() -> (InfoBox:WBP UIKit Info Box Object Reference), InText:Text=@n1)
  ]}
 UpdateInfoBoxVisual {nodes [
   UpdateInfoBoxVisual() {is=title K2Node_FunctionEntry}
   n1 = Get SB_Content()
   > SetMinDesiredHeight!(execute, self:Size Box Object Reference=@n1, InMinDesiredHeight:float=64) -> (then)
   > SetWidthOverride(self=@n1, InWidthOverride=444)
   n2 = Get InfoBox()
   > SetPadding!(
    execute
    self:Overlay Slot Object Reference=SlotAsOverlaySlot(Widget=Get HorizontalBox!(self:WBP UIKit Info Box Object Reference=@n2) -> (HorizontalBox:Horizontal Box Object Reference))
    InPadding:Margin Structure=Get ContentPadding() -> (ContentPadding:Margin Structure)
   ) -> (then)
   > RemoveFromParent(self=Get WBP_UIKit_Backplate!(self:WBP UIKit Info Box Object Reference=Get InfoBox()) -> (WBP_UIKit_Backplate:WBP UIKit Backplate Object Reference))
   n3 = Get Background() -> (Background:Image Object Reference)
   > SetBrush(
    self=@n3
    InBrush=MakeBrushFromMaterial!(
     Material:Material Interface Object Reference=/FortUILibrary/Systems/InputField/MaterialInstances/MI_UIKit_BackgroundRounded
     Width:int=32
     Height:int=32
    ) -> (ReturnValue:Slate Brush Structure)
   )
   > SetBrushTintColor!(
    execute
    self:Image Object Reference=@n3
    TintColor_SpecifiedColor:Linear Color Structure=(R=0,G=0,B=0,A=0.95)
    TintColor_ColorUseRule:ESlateColorStylingMode Enum
   ) -> (then)
   > n4 = GetDynamicMaterial(self=Get Background())
   n5 = Get LMaterialCornerRadius() -> (LMaterialCornerRadius:double)
   > SetVectorParameterValue(
    self=@n4
    ParameterName=Corner Radius
    Value=MakeColor!(R:float=@n5, G:float=@n5, B:float=@n5, A:float=@n5) -> (ReturnValue:Linear Color Structure)
   )
  ]}
]
variables [
 animationstate {type=integer}
 NotifyTooltipIsVisibleUpdated {type=boolean}
]
widgets {
 info {inheritedWidgetCount=0, parentClass=&UserWidget, rootWidgetClass=&SizeBox, widgetCount=4}
 widgets [
  SB_Content {
   widgetClassPath=&SizeBox
   bIsVariable=true
   widget=SB_Content
   properties {
    bOverride_WidthOverride=true
    minDesiredWidth=444
    visibility=Visible
    widthOverride=444
    renderTransform {angle=0, scale=<1 1>, shear=<0 0>, translation=<0 -2>}
   }
  }
  GridPanel_0 {
   widgetClassPath=&GridPanel
   parent=SB_Content
   widget=GridPanel_0
   properties {columnFill [1], rowFill [1]}
   slot {class=SizeBoxSlot}
  }
  InfoBox {
   widgetClassPath=/FortUILibrary/UIKit/InfoBox/WBP_UIKit_InfoBox
   bIsVariable=true
   parent=GridPanel_0
   widget=InfoBox
   properties {previewText=Any player will be able to see mutual friends on your profile, showIcon=false, textJustification=Center, use Preview Text=true}
   slot {class=GridSlot, properties {layer=1}}
  }
  Background {
   widgetClassPath=&Image
   bIsVariable=true
   parent=GridPanel_0
   widget=Background
   properties {renderOpacity=0.95, visibility=SelfHitTestInvisible}
   slot {class=GridSlot}
  }
 ]
}
