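# Narrow sidebar settings list entry for discrete options, showing the setting name and modified indicator beside a rotator-based value picker.
# On init, if a runtime option and the setting allow tooltips, it spawns a floating description widget with a tooltip view model in a visual attachment.
# Non-touch hover animates the row, rotator and indicator tint; non-editable entries dim the value panel.
# Arabic culture swaps the name and value columns and padding.
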WidgetBlueprint /FortniteCoreUI/UI/Sidebar/ListEntries/SidebarSettingsListEntry_Discrete_Narrow
animations [OnHover {
  binding_count=2
  length_seconds=0.1
  object=&self:OnHover
  sequence {bindings [
    HorizontalBox_1 {
     id=#1
     display_name=ScaleBox_0
     possessed_class=HorizontalBox
     tracks [{
       type=MovieScene2DTransformTrack
       sections [{
         type=MovieScene2DTransformSection
         channels [
          Translation.X {type=MovieSceneScriptingFloatChannel, keys [[0,0],[2,-8]]}
          Translation.Y {type=MovieSceneScriptingFloatChannel}
          Angle {type=MovieSceneScriptingFloatChannel}
          Scale.X {type=MovieSceneScriptingFloatChannel}
          Scale.Y {type=MovieSceneScriptingFloatChannel}
          Shear.X {type=MovieSceneScriptingFloatChannel}
          Shear.Y {type=MovieSceneScriptingFloatChannel}
         ]
        }]
      }]
    }
    Text_SettingName {
     id=#2
     possessed_class=CommonTextBlock
     tracks [
      {
       type=MovieSceneFloatTrack
       display_name=Render Opacity
       sections [{type=MovieSceneFloatSection, channels [None {type=MovieSceneScriptingFloatChannel, const [1,2]}]}]
      }
      {
       type=MovieSceneColorTrack
       display_name=Color and Opacity
       sections [{
         type=MovieSceneColorSection
         channels [
          Color.R {type=MovieSceneScriptingFloatChannel, keys [[0,1],[2,0]]}
          Color.G {type=MovieSceneScriptingFloatChannel, keys [[0,1],[2,0]]}
          Color.B {type=MovieSceneScriptingFloatChannel, keys [[0,1],[2,0]]}
          Color.A {type=MovieSceneScriptingFloatChannel, const [1,1]}
         ]
        }]
      }
     ]
    }
   ]}
 }]
graphs [
 EventGraph {nodes [
   n1 = Event OnEditableStateChanged!() -> (OutputDelegate, then, bIsEditable:bool) {is=title K2Node_Event}
   > SetRenderOpacity(
    self=Get Panel_Value() -> (Panel_Value:Panel Widget Object Reference)
    InOpacity=Select(Option 0:float=0.45, Option 1:float=1, Index=@n1.bIsEditable) -> (ReturnValue:float)
   )
   Event Construct()
   > Update Visual for Arabic()
   Event OnEntrySettingInitialized() {is=title K2Node_Event}
   > Branch(Condition=GetIsClamOneEnabled!(self:Fort Runtime Options Object Reference=GetRuntimeOptions()) -> (ReturnValue:bool))
   > Branch(Condition=GetCanShowTooltips!(self:Fort Setting Object Reference=GetEntrySetting!(self:Fort Settings List Entry Base Object Reference) -> (ReturnValue:Fort Setting Object Reference)) -> (ReturnValue:bool))
   > n2 = Create Widget!(
    execute:Exec
    Class:User Widget Class Reference=/FortniteCoreUI/UI/Sidebar/WBP_Sidebar_FloatingDescription.WBP_Sidebar_FloatingDescription_C
    OwningPlayer:Player Controller Object Reference
    FortFloatingTooltipViewModel:Fort Floating Tooltip View Model Object Reference
   ) -> (then:Exec, ReturnValue:WBP Sidebar Floating Description Object Reference)
   > AddChild(self=Get VA_Tooltip() -> (VA_Tooltip:Common Visual Attachment Object Reference), Content=@n2)
   n3 = SlotAsSizeBoxSlot!(Widget:Widget Object Reference=@n2) -> (ReturnValue:Size Box Slot Object Reference)
   > SetPadding!(
    execute
    self:Size Box Slot Object Reference=@n3
    InPadding_Left:float=0
    InPadding_Top:float=0
    InPadding_Right:float=120
    InPadding_Bottom:float=0
   ) -> (then)
   > SetHorizontalAlignment(self:Size Box Slot Object Reference=@n3, InHorizontalAlignment=HAlign_Right)
   > SetVerticalAlignment(self:Size Box Slot Object Reference=@n3, InVerticalAlignment=VAlign_Fill)
   > SetFortFloatingTooltipViewModel!(
    execute
    self:WBP Sidebar Floating Description Object Reference=@n2
    Viewmodel:Fort Floating Tooltip View Model Object Reference=CreateFortSettingTooltipViewModel!(
     InOwningObject:Object Reference=@n2
     InSetting:Fort Setting Object Reference=GetEntrySetting()
     None:%=Get FortSettingTooltipViewModel() -> (FortSettingTooltipViewModel:Fort Setting Tooltip View Model Object Reference)
    ) -> (ReturnValue:Fort Setting Tooltip View Model Object Reference)
   ) -> (then)
   Event OnMouseEnter()
   n4 = IsUsingTouch(PlayerController=GetOwningPlayer())
   > n5 = Branch(Condition=@n4)
   n6 = Get OnHover() -> (OnHover:Widget Animation Object Reference)
   PlayAnimationForward(execute=@n5.else, InAnimation=@n6)
   > n7 = Cast To SettingsRotator_Narrow!(execute, Object=Get Rotator_SettingValue() -> (Rotator_SettingValue:Fort Settings Rotator Object Reference)) -> (then, CastFailed, AsSettings Rotator Narrow:Settings Rotator Narrow Object Reference) {is=title K2Node_DynamicCast}
   > OnHover Animation(self:Settings Rotator Narrow Object Reference=@n7)
   > SetColorAndOpacity(
    self:Image Object Reference=Get OptionsModifiedIndicator() -> (OptionsModifiedIndicator:Image Object Reference)
    InColorAndOpacity:Linear Color Structure=(R=0.0513,G=0.0612,B=0.0723,A=1)
   )
   Event OnMouseLeave!() -> (OutputDelegate, then, MouseEvent:"Pointer Event Structure (by ref)") {is=title K2Node_Event}
   > n8 = Branch(Condition=@n4)
   PlayAnimationReverse(execute=@n8.else, InAnimation=@n6)
   > n9 = Cast To SettingsRotator_Narrow(Object=Get Rotator_SettingValue())
   > OnUnHover Animation(self:Settings Rotator Narrow Object Reference=@n9)
   > SetColorAndOpacity(self=Get OptionsModifiedIndicator(), InColorAndOpacity=(R=0.5089,G=0.5583,B=0.7084,A=1))
  ]}
 GetPrimaryGamepadFocusWidget {nodes [
   GetPrimaryGamepadFocusWidget() {is=title K2Node_FunctionEntry}
   > Return Node!(execute, ReturnValue:Widget Object Reference=Get Rotator_SettingValue())
  ]}
 Update Visual for Arabic {nodes [
   Update Visual for Arabic() {is=title K2Node_FunctionEntry}
   n1 = "Equal Exactly (String)"(A=GetCurrentCulture!() -> (ReturnValue:String), B=ar)
   > SetColumn!(
    execute
    self:Grid Slot Object Reference=SlotAsGridSlot(Widget=Get SizeBox() -> (SizeBox:Size Box Object Reference))
    InColumn:int=Select(Option 0:int=0, Option 1:int=1, Index=@n1) -> (ReturnValue:int)
   ) -> (then)
   > SetColumn(self=SlotAsGridSlot(Widget=Get Panel_Value()), InColumn=Select(Option 0=1, Option 1=0, Index=@n1))
   n2 = SlotAsHorizontalBoxSlot!(Widget:Widget Object Reference=Get OptionsSets() -> (OptionsSets:Widget Switcher Object Reference)) -> (ReturnValue:Horizontal Box Slot Object Reference)
   n3 = Get!(self:Horizontal Box Slot Object Reference=@n2) -> (Padding_Left:float=~, Padding_Top:float=~, Padding_Right:float=~, Padding_Bottom:float=~)
   > SetPadding(
    self:Horizontal Box Slot Object Reference=@n2
    InPadding_Left=Select(Option 0:float=0, Option 1:float=20, Index=@n1) -> (ReturnValue:float)
    InPadding_Top=@n3.Padding_Top
    InPadding_Right=Select(Option 0=20, Option 1=0, Index=@n1)
    InPadding_Bottom=@n3.Padding_Bottom
   )
  ]}
]
variables []
widgets {
 info {
  inheritedWidgetCount=8
  namedSlotCount=1
  parentClass=/Script/FortniteUI.FortSettingsListEntrySetting_Discrete
  rootWidgetClass=&GridPanel
  widgetCount=22
 }
 widgets [
  GridPanel_0 {widgetClassPath=&GridPanel, widget=GridPanel_0, properties {columnFill [0,1], rowFill [1]}}
  SidebarSettingEntryBackground {
   widgetClassPath=&/SidebarSettingEntryBackground
   bIsVariable=true
   parent=GridPanel_0
   widget=SidebarSettingEntryBackground
   slot {class=GridSlot, properties {column=1}}
  }
  GridPanel {widgetClassPath=&GridPanel, namedSlotHost=SidebarSettingEntryBackground, widget=GridPanel, properties {columnFill [1,1]}}
  SizeBox {
   widgetClassPath=&SizeBox
   bIsVariable=true
   parent=GridPanel
   widget=SizeBox
   properties {bOverride_MinDesiredHeight=true, minDesiredHeight=50}
   slot {class=GridSlot}
  }
  HorizontalBox_0 {
   widgetClassPath=&HorizontalBox
   parent=SizeBox
   widget=HorizontalBox_0
   slot {class=SizeBoxSlot, properties {padding=<ltrb 20 0 20 0>}}
  }
  ScaleBox_0 {
   widgetClassPath=&ScaleBox
   parent=HorizontalBox_0
   widget=ScaleBox_0
   properties {stretchDirection=DownOnly}
   slot {class=HorizontalBoxSlot, properties {size {sizeRule=Fill, value=1}}}
  }
  Text_SettingName {
   widgetClassPath=&CommonTextBlock
   bInherited=true
   bIsVariable=true
   parent=ScaleBox_0
   widget=Text_SettingName
   properties {
    text="{Setting Name}"
    textTransformPolicy=ToUpper
    font {
     bForceMonospaced=false
     bMaterialIsStencil=false
     fontMaterial=null
     fontObject=&BurbankSmall
     letterSpacing=0
     monospacedWidth=1
     size=18
     skewAmount=0
     typefaceFontName=Black
     outlineSettings {
      bApplyOutlineToDropShadows=false
      bMiteredCorners=false
      bSeparateFillAlpha=false
      outlineMaterial=null
      outlineSize=0
      outlineColor=<rgba 0 0 0 1>
     }
    }
    shadowOffset=<0 0>
    shapedTextOptions {bOverride_TextFlowDirection=true, bOverride_TextShapingMethod=false, textFlowDirection=Culture, textShapingMethod=Auto}
   }
   slot {class=ScaleBoxSlot, properties {horizontalAlignment=HAlign_Fill}}
  }
  OptionsModifiedIndicator {
   widgetClassPath=&Image
   bInherited=true
   bIsVariable=true
   parent=HorizontalBox_0
   widget=OptionsModifiedIndicator
   properties {
    brush {
     bIsDynamicallyLoaded=false
     drawAs=Image
     imageType=NoImage
     mirroring=NoMirror
     resourceName=None
     resourceObject=/Game/UI/Foundation/Textures/Icons/Audio/T_UI_Icon_VoiceParentalRestricted
     tiling=NoTile
     imageSize=<32 32>
     margin=<ltrb 0 0 0 0>
     outlineSettings {
      bUseBrushTransparency=false
      roundingType=HalfHeightRadius
      width=0
      color {colorUseRule=UseColor_Specified, specifiedColor=<rgba 0 0 0 0>}
      cornerRadii=<0 0 0 0>
     }
     tintColor {colorUseRule=UseColor_Specified, specifiedColor=<rgba 1 1 1 1>}
     uVRegion {bIsValid=false, max=<0 0>, min=<0 0>}
    }
    colorAndOpacity=<rgba 0.5089 0.5583 0.7084 1>
   }
   slot {class=HorizontalBoxSlot, properties {horizontalAlignment=HAlign_Center, verticalAlignment=VAlign_Center}}
  }
  Panel_Value {
   widgetClassPath=&HorizontalBox
   bInherited=true
   bIsVariable=true
   parent=GridPanel
   widget=Panel_Value
   slot {class=GridSlot, properties {column=1}}
  }
  OptionsSets {
   widgetClassPath=&WidgetSwitcher
   bIsVariable=true
   parent=Panel_Value
   widget=OptionsSets
   properties {activeWidgetIndex=1}
   slot {class=HorizontalBoxSlot, properties {padding=<ltrb 0 5 20 5>, size {sizeRule=Fill, value=1}}}
  }
  Couple_Options {widgetClassPath=&HorizontalBox, parent=OptionsSets, widget=Couple_Options, slot {class=WidgetSwitcherSlot}}
  Button_FirstOption {
   widgetClassPath=&IconTextButton
   bIsVariable=true
   parent=Couple_Options
   widget=Button_FirstOption
   properties {
    bInteractableWhenSelected=true
    bToggleable=true
    buttonText=""
    onButtonBaseClicked=()
    onButtonBaseDoubleClicked=()
    onButtonBaseHovered=()
    onButtonBaseUnhovered=()
    onSelectedChangedBase=()
    style=/FortUILibrary/Foundation/Buttons/ButtonStyles/Athena/Athena_ButtonStyle_BlueMenuButton_Settings.Athena_ButtonStyle_BlueMenuButton_Settings_C
   }
   slot {class=HorizontalBoxSlot, properties {size {sizeRule=Fill, value=1}}}
  }
  Button_SecondOption {
   widgetClassPath=&IconTextButton
   bIsVariable=true
   parent=Couple_Options
   widget=Button_SecondOption
   properties {
    bInteractableWhenSelected=true
    bToggleable=true
    buttonText=""
    onButtonBaseClicked=()
    onButtonBaseDoubleClicked=()
    onButtonBaseHovered=()
    onButtonBaseUnhovered=()
    onSelectedChangedBase=()
    style=&Athena_ButtonStyle_BlueMenuButton_Settings_C
    renderTransform {angle=0, scale=<-1 1>, shear=<0 0>, translation=<0 0>}
   }
   slot {class=HorizontalBoxSlot, properties {size {sizeRule=Fill, value=1}}}
  }
  Several_Options {widgetClassPath=&HorizontalBox, bIsVariable=true, parent=OptionsSets, widget=Several_Options, slot {class=WidgetSwitcherSlot}}
  Button_Decrease {
   widgetClassPath=&IconTextButton
   bInherited=true
   bIsVariable=true
   parent=Several_Options
   widget=Button_Decrease
   properties {
    bInteractableWhenSelected=true
    bToggleable=true
    buttonText=""
    onButtonBaseClicked=()
    onButtonBaseDoubleClicked=()
    onButtonBaseHovered=()
    onButtonBaseUnhovered=()
    onSelectedChangedBase=()
    style=&Athena_ButtonStyle_BlueMenuButton_Settings_C
    touchMethod=PreciseTap
    visibility=Collapsed
    iconBrush {
     bIsDynamicallyLoaded=false
     drawAs=Image
     imageType=NoImage
     mirroring=NoMirror
     resourceName=None
     resourceObject=/Game/UI/SocialPanel/Art/Textures/T_UI_Icon_TreeArrow_64
     tiling=NoTile
     imageSize=<32 32>
     margin=<ltrb 0 0 0 0>
     outlineSettings {
      bUseBrushTransparency=false
      roundingType=HalfHeightRadius
      width=0
      color {colorUseRule=UseColor_Specified, specifiedColor=<rgba 0 0 0 0>}
      cornerRadii=<0 0 0 1>
     }
     tintColor {colorUseRule=UseColor_Specified, specifiedColor=<rgba 0.0331 0.402 1 1>}
     uVRegion {bIsValid=false, max=<0 0>, min=<0 0>}
    }
    renderTransform {angle=0, scale=<-1 1>, shear=<0 0>, translation=<0 0>}
   }
   slot {class=HorizontalBoxSlot}
  }
  Overlay_0 {
   widgetClassPath=&Overlay
   parent=Several_Options
   widget=Overlay_0
   slot {class=HorizontalBoxSlot, properties {size {sizeRule=Fill, value=1}}}
  }
  Border_0 {
   widgetClassPath=&Border
   parent=Overlay_0
   widget=Border_0
   properties {verticalAlignment=VAlign_Center, brushColor=<rgba 0.0037 0.0052 0.016 1>, padding=<ltrb 0 2 0 2>}
   slot {class=OverlaySlot, properties {horizontalAlignment=HAlign_Fill, verticalAlignment=VAlign_Fill}}
  }
  Rotator_SettingValue {
   widgetClassPath=&/SidebarSettingsRotator_Narrow
   bInherited=true
   bIsVariable=true
   parent=Border_0
   widget=Rotator_SettingValue
   slot {class=BorderSlot, properties {verticalAlignment=VAlign_Center, padding=<ltrb 0 2 0 2>}}
  }
  Button_Increase {
   widgetClassPath=&IconTextButton
   bInherited=true
   bIsVariable=true
   parent=Overlay_0
   widget=Button_Increase
   properties {
    bInteractableWhenSelected=true
    bToggleable=true
    buttonText=""
    onButtonBaseClicked=()
    onButtonBaseDoubleClicked=()
    onButtonBaseHovered=()
    onButtonBaseUnhovered=()
    onSelectedChangedBase=()
    style=/FortUILibrary/Foundation/Buttons/ButtonStyles/ButtonStyle-Clear.ButtonStyle-Clear_C
    touchMethod=PreciseTap
    useLightInput=true
    iconBrush {
     bIsDynamicallyLoaded=false
     drawAs=NoDrawType
     imageType=NoImage
     mirroring=NoMirror
     resourceName=None
     resourceObject=null
     tiling=NoTile
     imageSize=<0 0>
     margin=<ltrb 0 0 0 0>
     outlineSettings {
      bUseBrushTransparency=false
      roundingType=HalfHeightRadius
      width=0
      color {colorUseRule=UseColor_Specified, specifiedColor=<rgba 0 0 0 0>}
      cornerRadii=<0 0 0 1>
     }
     tintColor {colorUseRule=UseColor_Foreground, specifiedColor=<rgba 0.0331 0.402 1 1>}
     uVRegion {bIsValid=false, max=<0 0>, min=<0 0>}
    }
   }
   slot {class=OverlaySlot, properties {horizontalAlignment=HAlign_Fill, verticalAlignment=VAlign_Fill}}
  }
  VA_Tooltip {
   widgetClassPath=&CommonVisualAttachment
   bIsVariable=true
   parent=GridPanel_0
   widget=VA_Tooltip
   properties {contentAnchor=<1 0>}
   slot {class=GridSlot}
  }
  Button_SelectEntry {widgetClassPath=&CommonButtonLegacy, bInherited=true, widget=null}
  Background {widgetClassPath=&UserWidget, bInherited=true, widget=null}
 ]
}
