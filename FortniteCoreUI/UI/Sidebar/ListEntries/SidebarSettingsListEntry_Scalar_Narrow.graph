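# Narrow sidebar settings list entry for a scalar setting, showing its name beside a slider-style value bar.
# An invisible AnalogSlider, returned as gamepad focus, overlays a material progress bar and value text; OnValueChanged drives the bar.
# OnDefaultValueChanged writes a material parameter on a default-value image, which starts collapsed.
# Mouse enter and leave play and reverse the OnHover color, opacity and shift animation, skipped for touch input.
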WidgetBlueprint /FortniteCoreUI/UI/Sidebar/ListEntries/SidebarSettingsListEntry_Scalar_Narrow
animations [OnHover {
  binding_count=5
  length_seconds=0.1
  object=&self:OnHover
  sequence {bindings [
    Progress {
     id=#1
     possessed_class=SimpleMaterialProgressBar_C
     tracks [
      {
       type=MovieSceneColorTrack
       display_name=Color A
       sections [{
         type=MovieSceneColorSection
         channels [
          Color.R {type=MovieSceneScriptingFloatChannel, keys [[0,0.0185],[1.9997,0.0382]]}
          Color.G {type=MovieSceneScriptingFloatChannel, keys [[0,0.1746],[1.9997,0.4452]]}
          Color.B {type=MovieSceneScriptingFloatChannel, keys [[0,0.5647],[1.9997,0.9047]]}
          Color.A {type=MovieSceneScriptingFloatChannel, const [1,1]}
         ]
        }]
      }
      {
       type=MovieSceneColorTrack
       display_name=Color B
       sections [{
         type=MovieSceneColorSection
         channels [
          Color.R {type=MovieSceneScriptingFloatChannel, keys [[0,0.0185],[2,0.0382]]}
          Color.G {type=MovieSceneScriptingFloatChannel, keys [[0,0.1746],[2,0.4452]]}
          Color.B {type=MovieSceneScriptingFloatChannel, keys [[0,0.5647],[2,0.9047]]}
          Color.A {type=MovieSceneScriptingFloatChannel, const [1,1]}
         ]
        }]
      }
     ]
    }
    Text_SettingName {
     id=#2
     possessed_class=CommonTextBlock
     tracks [
      {
       type=MovieSceneFloatTrack
       display_name=Render Opacity
       sections [{type=MovieSceneFloatSection, channels [None {type=MovieSceneScriptingFloatChannel, const [1,2]}]}]
      }
      {
       type=MovieSceneColorTrack
       display_name=Color and Opacity
       sections [{
         type=MovieSceneColorSection
         channels [
          Color.R {type=MovieSceneScriptingFloatChannel, keys [[0,1],[2,0]]}
          Color.G {type=MovieSceneScriptingFloatChannel, keys [[0,1],[2,0]]}
          Color.B {type=MovieSceneScriptingFloatChannel, keys [[0,1],[2,0]]}
          Color.A {type=MovieSceneScriptingFloatChannel, const [1,1]}
         ]
        }]
      }
     ]
    }
    Text_SettingValue {
     id=#3
     possessed_class=CommonTextBlock
     tracks [{
       type=MovieSceneFloatTrack
       display_name=Render Opacity
       sections [{type=MovieSceneFloatSection, channels [None {type=MovieSceneScriptingFloatChannel, keys [[0,0.5],[1.9997,1]]}]}]
      }]
    }
    Border_OnlineStatusIndicator {
     id=#4
     possessed_class=Border
     tracks [{
       type=MovieSceneColorTrack
       display_name=Brush Color
       sections [{
         type=MovieSceneColorSection
         channels [
          Color.R {type=MovieSceneScriptingFloatChannel, const [0,2]}
          Color.G {type=MovieSceneScriptingFloatChannel, keys [[0,0.1559],[1.9997,0]]}
          Color.B {type=MovieSceneScriptingFloatChannel, keys [[0,0.4452],[1.9997,0]]}
          Color.A {type=MovieSceneScriptingFloatChannel, const [1,1]}
         ]
        }]
      }]
    }
    HorizontalBox_1 {
     id=#5
     possessed_class=HorizontalBox
     tracks [{
       type=MovieScene2DTransformTrack
       sections [{
         type=MovieScene2DTransformSection
         channels [
          Translation.X {type=MovieSceneScriptingFloatChannel, keys [[0,0],[1.9997,-8]]}
          Translation.Y {type=MovieSceneScriptingFloatChannel}
          Angle {type=MovieSceneScriptingFloatChannel}
          Scale.X {type=MovieSceneScriptingFloatChannel}
          Scale.Y {type=MovieSceneScriptingFloatChannel}
          Shear.X {type=MovieSceneScriptingFloatChannel}
          Shear.Y {type=MovieSceneScriptingFloatChannel}
         ]
        }]
      }]
    }
   ]}
 }]
graphs [
 EventGraph {nodes [
   n1 = Event OnValueChanged!() -> (OutputDelegate, then, Value:float) {is=title K2Node_Event}
   > SetProgress!(
    execute
    self:Fort Simple Material Progress Bar Object Reference=Get Progress() -> (Progress:Simple Material Progress Bar Object Reference)
    Progress:float=@n1.Value
   ) -> (then)
   n2 = Event OnDefaultValueChanged!() -> (OutputDelegate, then, DefaultValue:float) {is=title K2Node_Event}
   > n3 = GetDynamicMaterial(self=Get DefaultValueLayer() -> (DefaultValueLayer:Image Object Reference))
   > SetScalarParameterValue(self=@n3, ParameterName=DefaultValue, Value=@n2.DefaultValue)
   Event OnMouseEnter()
   n4 = IsUsingTouch(PlayerController=GetOwningPlayer())
   > n5 = Branch(Condition=@n4)
   n6 = Get OnHover() -> (OnHover:Widget Animation Object Reference)
   PlayAnimationForward(execute=@n5.else, InAnimation=@n6)
   Event OnMouseLeave!() -> (OutputDelegate, then, MouseEvent:"Pointer Event Structure (by ref)") {is=title K2Node_Event}
   > n7 = Branch(Condition=@n4)
   PlayAnimationReverse(execute=@n7.else, InAnimation=@n6)
  ]}
 GetPrimaryGamepadFocusWidget {nodes [
   GetPrimaryGamepadFocusWidget() {is=title K2Node_FunctionEntry}
   > Return Node!(execute, ReturnValue:Widget Object Reference=Get Slider_SettingValue() -> (Slider_SettingValue:Analog Slider Object Reference))
  ]}
]
variables []
widgets {
 info {
  inheritedWidgetCount=5
  namedSlotCount=1
  parentClass=/Script/FortniteUI.FortSettingsListEntrySetting_Scalar
  rootWidgetClass=&Overlay
  widgetCount=16
 }
 widgets [
  Overlay_0 {widgetClassPath=&Overlay, widget=Overlay_0}
  SidebarSettingEntryBackground {
   widgetClassPath=&/SidebarSettingEntryBackground
   bIsVariable=true
   parent=Overlay_0
   widget=SidebarSettingEntryBackground
   slot {class=OverlaySlot, properties {horizontalAlignment=HAlign_Fill, verticalAlignment=VAlign_Fill}}
  }
  FortResponsivePanel_1 {
   widgetClassPath=/Script/FortniteUI.FortResponsivePanel
   namedSlotHost=SidebarSettingEntryBackground
   widget=FortResponsivePanel_1
   properties {bCanStackVertically=false}
  }
  SizeBox_0 {
   widgetClassPath=&SizeBox
   parent=FortResponsivePanel_1
   widget=SizeBox_0
   properties {bOverride_MinDesiredHeight=true, minDesiredHeight=50}
   slot {class=FortResponsivePanelSlot}
  }
  HorizontalBox_1 {
   widgetClassPath=&HorizontalBox
   parent=SizeBox_0
   widget=HorizontalBox_1
   slot {class=SizeBoxSlot, properties {padding=<ltrb 20 0 0 0>}}
  }
  Border_OnlineStatusIndicator {
   widgetClassPath=&Border
   parent=HorizontalBox_1
   widget=Border_OnlineStatusIndicator
   properties {renderOpacity=0.5, visibility=Collapsed, brushColor=<rgba 0 0.1559 0.4452 1>, padding=<ltrb 0 0 0 0>}
   slot {class=HorizontalBoxSlot}
  }
  Spacer_0 {
   widgetClassPath=&Spacer
   parent=Border_OnlineStatusIndicator
   widget=Spacer_0
   properties {size=<8 8>}
   slot {class=BorderSlot, properties {padding=<ltrb 0 0 0 0>}}
  }
  ScaleBox_0 {
   widgetClassPath=&ScaleBox
   parent=HorizontalBox_1
   widget=ScaleBox_0
   slot {class=HorizontalBoxSlot, properties {verticalAlignment=VAlign_Center, padding=<ltrb 0 0 20 0>, size {sizeRule=Fill, value=1}}}
  }
  Text_SettingName {
   widgetClassPath=&CommonTextBlock
   bInherited=true
   bIsVariable=true
   parent=ScaleBox_0
   widget=Text_SettingName
   properties {
    text="{Setting Name}"
    textTransformPolicy=ToUpper
    font {
     bForceMonospaced=false
     bMaterialIsStencil=false
     fontMaterial=null
     fontObject=&BurbankSmall
     letterSpacing=0
     monospacedWidth=1
     size=18
     skewAmount=0
     typefaceFontName=Black
     outlineSettings {
      bApplyOutlineToDropShadows=false
      bMiteredCorners=false
      bSeparateFillAlpha=false
      outlineMaterial=null
      outlineSize=0
      outlineColor=<rgba 0 0 0 1>
     }
    }
    shadowOffset=<0 0>
   }
   slot {class=ScaleBoxSlot, properties {horizontalAlignment=HAlign_Fill}}
  }
  SizeBox_1 {
   widgetClassPath=&SizeBox
   parent=FortResponsivePanel_1
   widget=SizeBox_1
   properties {bOverride_MinDesiredHeight=true, minDesiredHeight=50}
   slot {class=FortResponsivePanelSlot}
  }
  Panel_Value {
   widgetClassPath=&Overlay
   bInherited=true
   bIsVariable=true
   parent=SizeBox_1
   widget=Panel_Value
   slot {class=SizeBoxSlot, properties {padding=<ltrb 0 5 20 5>}}
  }
  Progress {
   widgetClassPath=/Game/Athena/HUD/SimpleMaterialProgressBar
   bIsVariable=true
   parent=Panel_Value
   widget=Progress
   properties {
    bOverrideDefaultColorA=true
    bOverrideDefaultColorB=true
    bOverrideDefaultColorBackground=true
    bUseStroke=false
    visibility=HitTestInvisible
    cachedColorA=<rgba 0.0021 0.0612 0.5029 1>
    cachedColorB=<rgba 0.0021 0.0612 0.5029 1>
    cachedColorBackground=<rgba 0.0037 0.0052 0.016 1>
    foregroundColor {colorUseRule=UseColor_Specified, specifiedColor=<rgba 1 0 1 1>}
   }
   slot {class=OverlaySlot, properties {horizontalAlignment=HAlign_Fill, verticalAlignment=VAlign_Fill}}
  }
  DefaultValueLayer {
   widgetClassPath=&Image
   bIsVariable=true
   parent=Panel_Value
   widget=DefaultValueLayer
   properties {
    renderOpacity=0
    visibility=Collapsed
    brush {
     bIsDynamicallyLoaded=false
     drawAs=Image
     imageType=NoImage
     mirroring=NoMirror
     resourceName=None
     resourceObject=/Game/Athena/Settings/Art/SettingsDefaultSlider_M
     tiling=NoTile
     imageSize=<32 32>
     margin=<ltrb 0 0 0 0>
     outlineSettings {
      bUseBrushTransparency=false
      roundingType=HalfHeightRadius
      width=0
      color {colorUseRule=UseColor_Specified, specifiedColor=<rgba 0 0 0 0>}
      cornerRadii=<0 0 0 0>
     }
     tintColor {colorUseRule=UseColor_Specified, specifiedColor=<rgba 1 1 1 1>}
     uVRegion {bIsValid=false, max=<0 0>, min=<0 0>}
    }
   }
   slot {class=OverlaySlot, properties {horizontalAlignment=HAlign_Fill, verticalAlignment=VAlign_Fill}}
  }
  Text_SettingValue {
   widgetClassPath=&CommonTextBlock
   bInherited=true
   bIsVariable=true
   parent=Panel_Value
   widget=Text_SettingValue
   properties {
    justification=Center
    renderOpacity=0.5
    scrollStyle=&TextScrollStyle-Base_C
    text="{Value}"
    visibility=HitTestInvisible
    font {
     bForceMonospaced=false
     bMaterialIsStencil=false
     fontMaterial=null
     fontObject=&BurbankSmall
     letterSpacing=0
     monospacedWidth=1
     size=16
     skewAmount=0
     typefaceFontName=Black
     outlineSettings {
      bApplyOutlineToDropShadows=false
      bMiteredCorners=false
      bSeparateFillAlpha=false
      outlineMaterial=null
      outlineSize=0
      outlineColor=<rgba 0 0 0 1>
     }
    }
    shadowOffset=<0 0>
   }
   slot {class=OverlaySlot, properties {horizontalAlignment=HAlign_Fill, verticalAlignment=VAlign_Center, padding=<ltrb 0 15 0 15>}}
  }
  Slider_SettingValue {
   widgetClassPath=/Script/CommonUI.AnalogSlider
   bInherited=true
   bIsVariable=true
   parent=Panel_Value
   widget=Slider_SettingValue
   properties {sliderBarColor=<rgba 1 1 1 0>, sliderHandleColor=<rgba 1 1 1 0>}
   slot {class=OverlaySlot, properties {horizontalAlignment=HAlign_Fill, verticalAlignment=VAlign_Fill}}
  }
  Background {widgetClassPath=&UserWidget, bInherited=true, widget=null}
 ]
}
