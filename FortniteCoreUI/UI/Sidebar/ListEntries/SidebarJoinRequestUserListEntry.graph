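# Sidebar list entry widget for a join-request user, built on FortSocialUserListEntry, with avatar, social name, rich presence and lock icon.
# Hover plays a highlight animation and switches name and presence to hover-styled text; unhover or focus loss restores them.
# Opening or closing the UserActionMenu anchor plays Selected forward or faster in reverse, with a sound.
# The new-avatars flag toggles PlayerAvatar against the legacy icon and online indicator.
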WidgetBlueprint /FortniteCoreUI/UI/Sidebar/ListEntries/SidebarJoinRequestUserListEntry
animations [
 Hovered {
  binding_count=5
  display_label=""
  length_seconds=0.2
  object=&self:Hovered
  sequence {bindings [
    BorderSlot_2 {
     id=#1
     parent_id=#2
     possessed_class=BorderSlot
     tracks [{
       type=MovieSceneMarginTrack
       display_name=Padding
       sections [{
         type=MovieSceneMarginSection
         end_frame=2
         start_frame=0
         channels [
          Left {type=MovieSceneScriptingFloatChannel}
          Top {type=MovieSceneScriptingFloatChannel}
          Right {type=MovieSceneScriptingFloatChannel, keys [[0,-32],[2,15]]}
          Bottom {type=MovieSceneScriptingFloatChannel}
         ]
        }]
      }]
    }
    SizeBox_2 {id=#3, possessed_class=SizeBox, tracks []}
    HorizontalBoxSlot_1 {
     id=#4
     parent_id=#3
     possessed_class=HorizontalBoxSlot
     tracks [{
       type=MovieSceneMarginTrack
       display_name=Padding
       sections [{
         type=MovieSceneMarginSection
         end_frame=3
         start_frame=0
         channels [
          Left {type=MovieSceneScriptingFloatChannel, keys [[0,10],[3,15]]}
          Top {type=MovieSceneScriptingFloatChannel}
          Right {type=MovieSceneScriptingFloatChannel}
          Bottom {type=MovieSceneScriptingFloatChannel}
         ]
        }]
      }]
    }
    InputActionWidget {
     id=#2
     possessed_class=CommonActionWidget
     tracks [{
       type=MovieSceneFloatTrack
       display_name=Render Opacity
       sections [{type=MovieSceneFloatSection, end_frame=4, start_frame=0, channels [None {type=MovieSceneScriptingFloatChannel, keys [[0,0],[2,1]]}]}]
      }]
    }
    Border_Background {
     id=#5
     possessed_class=Border
     tracks [{
       type=MovieSceneWidgetMaterialTrack
       sections [{
         type=MovieSceneParameterSection
         end_frame=4
         start_frame=0
         channels [AnimateHighlight {type=MovieSceneScriptingFloatChannel, keys [[0,0],[3,1],[4,1]]}]
        }]
      }]
    }
   ]}
 }
 Selected {
  binding_count=5
  display_label=""
  length_seconds=0.3
  object=&self:Selected
  sequence {
   bindings [
    Arrow_Select {
     id=#6
     possessed_class=Image
     tracks [
      {
       type=MovieSceneFloatTrack
       display_name=Render Opacity
       sections [{
         type=MovieSceneFloatSection
         end_frame=18
         start_frame=0
         channels [None {type=MovieSceneScriptingFloatChannel, keys [[0,0],[6,0],[15,1],[18,1]]}]
        }]
      }
      {
       type=MovieScene2DTransformTrack
       sections [{
         type=MovieScene2DTransformSection
         channels [
          Translation.X {type=MovieSceneScriptingFloatChannel, keys [[0,8],[6,8],[15,-6],[18,0]]}
          Translation.Y {type=MovieSceneScriptingFloatChannel}
          Angle {type=MovieSceneScriptingFloatChannel}
          Scale.X {type=MovieSceneScriptingFloatChannel, keys [[0,-0.875],[6,-0.875],[15,-1.125],[18,-1]]}
          Scale.Y {type=MovieSceneScriptingFloatChannel, keys [[0,0.875],[6,0.875],[15,1.125],[18,1]]}
          Shear.X {type=MovieSceneScriptingFloatChannel}
          Shear.Y {type=MovieSceneScriptingFloatChannel}
         ]
        }]
      }
     ]
    }
    Border_Background {
     id=#7
     possessed_class=Border
     tracks [
      {
       type=MovieSceneFloatTrack
       display_name=Render Opacity
       sections [{type=MovieSceneFloatSection, channels [None {type=MovieSceneScriptingFloatChannel, keys [[0,1],[12,0]]}]}]
      }
      {
       type=MovieScene2DTransformTrack
       sections [{
         type=MovieScene2DTransformSection
         channels [
          Translation.X {type=MovieSceneScriptingFloatChannel, keys [[0,0],[12,256]]}
          Translation.Y {type=MovieSceneScriptingFloatChannel}
          Angle {type=MovieSceneScriptingFloatChannel}
          Scale.X {type=MovieSceneScriptingFloatChannel}
          Scale.Y {type=MovieSceneScriptingFloatChannel}
          Shear.X {type=MovieSceneScriptingFloatChannel, keys [[0,0],[12,16]]}
          Shear.Y {type=MovieSceneScriptingFloatChannel}
         ]
        }]
      }
     ]
    }
    UserActionMenuInputButton {
     id=#8
     possessed_class=UserActionMenuInputButton_C
     tracks [{
       type=MovieSceneFloatTrack
       display_name=Render Opacity
       sections [{type=MovieSceneFloatSection, channels [None {type=MovieSceneScriptingFloatChannel, keys [[0,0],[12,0],[15,1]]}]}]
      }]
    }
    AnimatedInputBorder_Select {
     id=#9
     possessed_class=Border
     tracks [
      {
       type=MovieSceneColorTrack
       display_name=Brush Color
       sections [{
         type=MovieSceneColorSection
         channels [
          Color.R {type=MovieSceneScriptingFloatChannel}
          Color.G {type=MovieSceneScriptingFloatChannel}
          Color.B {type=MovieSceneScriptingFloatChannel}
          Color.A {type=MovieSceneScriptingFloatChannel, keys [[9,0],[18,1]]}
         ]
        }]
      }
      {
       type=MovieSceneWidgetMaterialTrack
       sections [{type=MovieSceneParameterSection, channels [AnimateStateChange {type=MovieSceneScriptingFloatChannel, keys [[6,0],[15,1],[18,1]]}]}]
      }
     ]
    }
    EmptyButtonToEatOneMouseUpEvent {
     id=#10
     possessed_class=IconTextButton_C
     tracks [{
       type=MovieSceneByteTrack
       display_name=Visibility
       sections [{
         type=MovieSceneByteSection
         end_frame=12
         start_frame=0
         channels [None {type=MovieSceneScriptingByteChannel, keys [[0,1],[0.055,0],[12,0]]}]
        }]
      }]
    }
   ]
   display_rate {numerator=60}
  }
 }
 Unhovered {
  binding_count=5
  length_seconds=0.2
  object=&self:Unhovered
  sequence {bindings [
    BorderSlot_2 {
     id=#1
     parent_id=#2
     possessed_class=BorderSlot
     tracks [{
       type=MovieSceneMarginTrack
       display_name=Padding
       sections [{
         type=MovieSceneMarginSection
         end_frame=4
         start_frame=0
         channels [
          Left {type=MovieSceneScriptingFloatChannel}
          Top {type=MovieSceneScriptingFloatChannel}
          Right {type=MovieSceneScriptingFloatChannel, keys [[2,15],[3,-5],[4,-32]]}
          Bottom {type=MovieSceneScriptingFloatChannel}
         ]
        }]
      }]
    }
    SizeBox_2 {id=#3, possessed_class=SizeBox, tracks []}
    HorizontalBoxSlot_1 {
     id=#4
     parent_id=#3
     possessed_class=HorizontalBoxSlot
     tracks [{
       type=MovieSceneMarginTrack
       display_name=Padding
       sections [{
         type=MovieSceneMarginSection
         end_frame=4
         start_frame=0
         channels [
          Left {type=MovieSceneScriptingFloatChannel, keys [[1,15],[4,10]]}
          Top {type=MovieSceneScriptingFloatChannel}
          Right {type=MovieSceneScriptingFloatChannel}
          Bottom {type=MovieSceneScriptingFloatChannel}
         ]
        }]
      }]
    }
    InputActionWidget {
     id=#2
     possessed_class=CommonActionWidget
     tracks [{
       type=MovieSceneFloatTrack
       display_name=Render Opacity
       sections [{type=MovieSceneFloatSection, end_frame=4, start_frame=0, channels [None {type=MovieSceneScriptingFloatChannel, keys [[2,1],[4,0]]}]}]
      }]
    }
    Border_Background {
     id=#5
     possessed_class=Border
     tracks [{
       type=MovieSceneWidgetMaterialTrack
       sections [{
         type=MovieSceneParameterSection
         end_frame=4
         start_frame=0
         channels [AnimateHighlight {type=MovieSceneScriptingFloatChannel, keys [[0,1],[1,1],[4,0]]}]
        }]
      }]
    }
   ]}
 }
]
graphs [EventGraph {nodes [
   n1 = "On Menu Open Changed (MenuAnchor_Actions)"!() -> (OutputDelegate, then, bIsOpen:bool) {is=title K2Node_ComponentBoundEvent}
   > n2 = Branch(Condition=@n1.bIsOpen)
   n3 = Get Selected() -> (Selected:Widget Animation Object Reference)
   > PlayAnimation(InAnimation=@n3)
   PlayAnimation(execute=@n2.else, InAnimation=@n3, PlayMode=Reverse, PlaybackSpeed=2)
   > Branch(Condition=IsAnimationPlaying(InAnimation=Get Selected()))
   > PlaySound2D(Sound=/Game/Sounds/Fort_UI_Sounds/FrontEnd/New_FrontEnd/Fortnite_UI_MainMenu_Generic_01_Cue)
   "On Input Icon Updated (InputActionWidget)"() {is=title K2Node_ComponentBoundEvent}
   > n4 = GetIconDynamicMaterial!(execute, self:Common Action Widget Object Reference=Get InputActionWidget() -> (InputActionWidget:Common Action Widget Object Reference)) -> (then, ReturnValue:Material Instance Dynamic Object Reference)
   > SetScalarParameterValue(self=@n4, ParameterName=Use Dark, Value=1)
   > SetScalarParameterValue(self=@n4, ParameterName=Outline Opacity, Value=0)
   n5 = Event OnFocusLost!() -> (OutputDelegate, then, InFocusEvent:Focus Event Structure) {is=title K2Node_Event}
   StopAllAnimations(execute=[@n5.then, @n6.then], self:User Widget Object Reference)
   > PlayAnimation(InAnimation=Get Unhovered() -> (Unhovered:Widget Animation Object Reference), PlaybackSpeed=2)
   > SetActiveWidget(
    self:Widget Switcher Object Reference=Get Switcher_SocialName() -> (Switcher_SocialName:Common Activatable Widget Switcher Object Reference)
    Widget=Get Text_SocialName() -> (Text_SocialName:Social Name Text Block Object Reference)
   )
   > SetActiveWidget(
    self=Get Switcher_RichPresence() -> (Switcher_RichPresence:Common Activatable Widget Switcher Object Reference)
    Widget=Get Text_RichPresence() -> (Text_RichPresence:Common Rich Text Block Object Reference)
   )
   Event On Hovered()
   > PlayAnimation(InAnimation=Get Hovered() -> (Hovered:Widget Animation Object Reference))
   > SetActiveWidget(self=Get Switcher_SocialName(), Widget=Get Text_SocialName_Hover() -> (Text_SocialName_Hover:Social Name Text Block Object Reference))
   > SetActiveWidget(self=Get Switcher_RichPresence(), Widget=Get Text_RichPresence_Hover() -> (Text_RichPresence_Hover:Common Rich Text Block Object Reference))
   n6 = Event On Unhovered()
   n7 = Event OnNewAvatarsEnabledChanged!() -> (OutputDelegate, then, bUsingNewAvatars:bool) {is=title K2Node_Event}
   > SetVisibility(
    self=Get PlayerAvatar() -> (PlayerAvatar:Fort Player Avatar Object Reference)
    InVisibility=Select(Option 0=Collapsed, Option 1=HitTestInvisible, Index=@n7.bUsingNewAvatars)
   )
   n8 = Select(Option 0=HitTestInvisible, Option 1=Collapsed, Index=@n7.bUsingNewAvatars)
   > SetVisibility(self=Get Avatar_MemberIcon() -> (Avatar_MemberIcon:Fort Social Avatar Icon Object Reference), InVisibility=@n8)
   > SetVisibility(self=Get Border_OnlineStatusIndicator() -> (Border_OnlineStatusIndicator:Border Object Reference), InVisibility=@n8)
  ]}]
variables [
 EncourageEpicFriend_HoveredColor {type=LinearColor, default=<rgba 1 1 1 1>}
 EncourageEpicFriend_IdleColor {type=LinearColor, default=<rgba 0.956 1 0.0666 1>}
]
widgets {
 info {inheritedWidgetCount=20, parentClass=/Script/FortniteUI.FortSocialUserListEntry, rootWidgetClass=&SizeBox, widgetCount=44}
 widgets [
  SizeBox_0 {widgetClassPath=&SizeBox, widget=SizeBox_0, properties {bOverride_MinDesiredHeight=true, heightOverride=72, minDesiredHeight=72}}
  Overlay_1 {widgetClassPath=&Overlay, parent=SizeBox_0, widget=Overlay_1, slot {class=SizeBoxSlot}}
  MenuAnchor_Actions {
   widgetClassPath=/Script/UMG.MenuAnchor
   bInherited=true
   bIsVariable=true
   parent=Overlay_1
   widget=MenuAnchor_Actions
   properties {
    menuClass=/Game/UI/SocialPanel/UserActionMenu.UserActionMenu_C
    placement=MenuPlacement_MenuRight
    renderTransform {angle=0, scale=<1 1>, shear=<0 0>, translation=<8 0>}
   }
   slot {class=OverlaySlot, properties {horizontalAlignment=HAlign_Right}}
  }
  Border_Background {
   widgetClassPath=&Border
   bInherited=true
   bIsVariable=true
   parent=Overlay_1
   widget=Border_Background
   properties {
    visibility=SelfHitTestInvisible
    background {
     bIsDynamicallyLoaded=false
     drawAs=Image
     imageType=NoImage
     mirroring=NoMirror
     resourceName=None
     resourceObject=/Game/UI/Sidebar/Materials/M_UI_sidebar_Button_Selection_Inst
     tiling=NoTile
     imageSize=<32 32>
     margin=<ltrb 0 0 0 0>
     outlineSettings {
      bUseBrushTransparency=false
      roundingType=HalfHeightRadius
      width=0
      color {colorUseRule=UseColor_Specified, specifiedColor=<rgba 0 0 0 0>}
      cornerRadii=<0 0 0 0>
     }
     tintColor {colorUseRule=UseColor_Specified, specifiedColor=<rgba 1 1 1 1>}
     uVRegion {bIsValid=false, max=<0 0>, min=<0 0>}
    }
    padding=<ltrb 0 0 0 0>
   }
   slot {class=OverlaySlot, properties {horizontalAlignment=HAlign_Fill, verticalAlignment=VAlign_Fill}}
  }
  HorizontalBox_2 {
   widgetClassPath=&HorizontalBox
   parent=Border_Background
   widget=HorizontalBox_2
   properties {flowDirectionPreference=Culture}
   slot {class=BorderSlot, properties {padding=<ltrb 0 0 0 0>}}
  }
  SizeBox_2 {
   widgetClassPath=&SizeBox
   parent=HorizontalBox_2
   widget=SizeBox_2
   properties {
    bOverride_MinDesiredHeight=true
    bOverride_MinDesiredWidth=true
    heightOverride=68
    minDesiredHeight=68
    minDesiredWidth=68
    widthOverride=68
   }
   slot {class=HorizontalBoxSlot, properties {horizontalAlignment=HAlign_Center, verticalAlignment=VAlign_Center, padding=<ltrb 10 10 10 10>}}
  }
  ScaleBox_0 {widgetClassPath=&ScaleBox, parent=SizeBox_2, widget=ScaleBox_0, properties {stretch=ScaleToFitY}, slot {class=SizeBoxSlot}}
  Overlay_2 {widgetClassPath=&Overlay, parent=ScaleBox_0, widget=Overlay_2, slot {class=ScaleBoxSlot}}
  Avatar_MemberIcon {
   widgetClassPath=/Script/FortniteUI.FortSocialAvatarIcon
   bInherited=true
   bIsVariable=true
   parent=Overlay_2
   widget=Avatar_MemberIcon
   properties {
    avatarMaterial=/Game/UI/SocialPanel/Art/Materials/UI_UserPortrait_Buffet
    materialTextureParamName=SmallPreview
    visibility=Collapsed
    brush {
     bIsDynamicallyLoaded=false
     drawAs=Image
     imageType=NoImage
     mirroring=NoMirror
     resourceName=None
     resourceObject=&UI_UserPortrait_Buffet
     tiling=NoTile
     imageSize=<32 32>
     margin=<ltrb 0 0 0 0>
     outlineSettings {
      bUseBrushTransparency=false
      roundingType=HalfHeightRadius
      width=0
      color {colorUseRule=UseColor_Specified, specifiedColor=<rgba 0 0 0 0>}
      cornerRadii=<0 0 0 0>
     }
     tintColor {colorUseRule=UseColor_Specified, specifiedColor=<rgba 1 1 1 1>}
     uVRegion {bIsValid=false, max=<0 0>, min=<0 0>}
    }
    loadingBackgroundBrush {
     bIsDynamicallyLoaded=false
     drawAs=Image
     imageType=NoImage
     mirroring=NoMirror
     resourceName=None
     resourceObject=/FortUILibrary/UIKit/Throbber/MI_UIKit_Monochrome_Throbber
     tiling=NoTile
     imageSize=<32 32>
     margin=<ltrb 0 0 0 0>
     outlineSettings {
      bUseBrushTransparency=false
      roundingType=HalfHeightRadius
      width=0
      color {colorUseRule=UseColor_Specified, specifiedColor=<rgba 0 0 0 0>}
      cornerRadii=<0 0 0 0>
     }
     tintColor {colorUseRule=UseColor_Specified, specifiedColor=<rgba 1 1 1 1>}
     uVRegion {bIsValid=false, max=<0 0>, min=<0 0>}
    }
   }
   slot {class=OverlaySlot}
  }
  Border_OnlineStatusIndicator {
   widgetClassPath=&Border
   bInherited=true
   bIsVariable=true
   parent=Overlay_2
   widget=Border_OnlineStatusIndicator
   properties {
    visibility=Collapsed
    background {
     bIsDynamicallyLoaded=false
     drawAs=Image
     imageType=NoImage
     mirroring=NoMirror
     resourceName=None
     resourceObject=/LauncherSocial/circle
     tiling=NoTile
     imageSize=<24 24>
     margin=<ltrb 0 0 0 0>
     outlineSettings {
      bUseBrushTransparency=false
      roundingType=HalfHeightRadius
      width=0
      color {colorUseRule=UseColor_Specified, specifiedColor=<rgba 0 0 0 0>}
      cornerRadii=<0 0 0 0>
     }
     tintColor {colorUseRule=UseColor_Specified, specifiedColor=<rgba 1 1 1 1>}
     uVRegion {bIsValid=false, max=<0 0>, min=<0 0>}
    }
    brushColor=<rgba 0.1946 0.9823 0.1812 1>
    padding=<ltrb 0 0 0 0>
   }
   slot {class=OverlaySlot, properties {horizontalAlignment=HAlign_Right, verticalAlignment=VAlign_Bottom}}
  }
  Spacer_0 {
   widgetClassPath=&Spacer
   parent=Border_OnlineStatusIndicator
   widget=Spacer_0
   properties {size=<8 8>}
   slot {class=BorderSlot, properties {padding=<ltrb 0 0 0 0>}}
  }
  PlayerAvatar {
   widgetClassPath=/Game/UI/Foundation/Profile/WBP_PlayerAvatar
   bInherited=true
   bIsVariable=true
   parent=Overlay_2
   widget=PlayerAvatar
   slot {class=OverlaySlot}
  }
  UserTextInfoVB {
   widgetClassPath=&VerticalBox
   parent=HorizontalBox_2
   widget=UserTextInfoVB
   properties {flowDirectionPreference=Culture}
   slot {class=HorizontalBoxSlot, properties {verticalAlignment=VAlign_Center, padding=<ltrb 10 10 15 10>, size {sizeRule=Fill, value=1}}}
  }
  HBox_SocialName {
   widgetClassPath=&HorizontalBox
   parent=UserTextInfoVB
   widget=HBox_SocialName
   slot {class=VerticalBoxSlot, properties {verticalAlignment=VAlign_Center}}
  }
  WBP_RebootRallyIcon {
   widgetClassPath=/Game/UI/Sidebar/RebootRally/WBP_RebootRallyIcon
   bInherited=true
   bIsVariable=true
   parent=HBox_SocialName
   widget=WBP_RebootRallyIcon
   properties {visibility=Collapsed}
   slot {class=HorizontalBoxSlot, properties {horizontalAlignment=HAlign_Left, verticalAlignment=VAlign_Center, padding=<ltrb -3 -3 0 -3>}}
  }
  Switcher_SocialName {
   widgetClassPath=&CommonActivatableWidgetSwitcher
   bIsVariable=true
   parent=HBox_SocialName
   widget=Switcher_SocialName
   properties {transitionCurveType=Linear, transitionDuration=0.1}
   slot {class=HorizontalBoxSlot, properties {verticalAlignment=VAlign_Center, size {sizeRule=Fill, value=1}}}
  }
  Text_SocialName {
   widgetClassPath=/Script/FortniteUI.FortSocialNameTextBlock
   bInherited=true
   bIsVariable=true
   parent=Switcher_SocialName
   widget=Text_SocialName
   properties {
    bOverrideDefaultStyle=true
    bTintInlineIcon=true
    mobileTextBlockScale=1.2
    platformIconDisplayRule=AlwaysIfDifferent
    scrollStyle=&TextScrollStyle-Base_C
    secondNameFontSize=14
    text="Nickname <PC color=\"82BDBDFF\" size=\"12\"></><text color=\"82BDBDFF\" size=\"12\">Display Name</>"
    defaultTextStyleOverride {
     overflowPolicy=Clip
     transformPolicy=None
     colorAndOpacity {colorUseRule=UseColor_Specified, specifiedColor=<rgba 1 1 1 1>}
     font {
      bForceMonospaced=false
      bMaterialIsStencil=false
      fontMaterial=null
      fontObject=&BurbankSmall
      letterSpacing=0
      monospacedWidth=1
      size=16
      skewAmount=0
      typefaceFontName=Black
      outlineSettings {
       bApplyOutlineToDropShadows=false
       bMiteredCorners=false
       bSeparateFillAlpha=false
       outlineMaterial=null
       outlineSize=0
       outlineColor=<rgba 0 0 0 1>
      }
     }
     highlightColor {colorUseRule=UseColor_Specified, specifiedColor=<rgba 0 0 0 1>}
     highlightShape {
      bIsDynamicallyLoaded=false
      drawAs=Image
      imageType=NoImage
      mirroring=NoMirror
      resourceName=None
      resourceObject=null
      tiling=NoTile
      imageSize=<32 32>
      margin=<ltrb 0 0 0 0>
      outlineSettings {
       bUseBrushTransparency=false
       roundingType=HalfHeightRadius
       width=0
       color {colorUseRule=UseColor_Specified, specifiedColor=<rgba 0 0 0 0>}
       cornerRadii=<0 0 0 0>
      }
      tintColor {colorUseRule=UseColor_Specified, specifiedColor=<rgba 1 1 1 1>}
      uVRegion {bIsValid=false, max=<0 0>, min=<0 0>}
     }
     selectedBackgroundColor {colorUseRule=UseColor_Foreground, specifiedColor=<rgba 1 0 1 1>}
     shadowColorAndOpacity=<rgba 0 0 0 1>
     shadowOffset=<0 0>
     strikeBrush {
      bIsDynamicallyLoaded=false
      drawAs=Image
      imageType=NoImage
      mirroring=NoMirror
      resourceName=None
      resourceObject=null
      tiling=NoTile
      imageSize=<32 32>
      margin=<ltrb 0 0 0 0>
      outlineSettings {
       bUseBrushTransparency=false
       roundingType=HalfHeightRadius
       width=0
       color {colorUseRule=UseColor_Specified, specifiedColor=<rgba 0 0 0 0>}
       cornerRadii=<0 0 0 0>
      }
      tintColor {colorUseRule=UseColor_Specified, specifiedColor=<rgba 1 1 1 1>}
      uVRegion {bIsValid=false, max=<0 0>, min=<0 0>}
     }
     underlineBrush {
      bIsDynamicallyLoaded=false
      drawAs=Image
      imageType=NoImage
      mirroring=NoMirror
      resourceName=None
      resourceObject=null
      tiling=NoTile
      imageSize=<32 32>
      margin=<ltrb 0 0 0 0>
      outlineSettings {
       bUseBrushTransparency=false
       roundingType=HalfHeightRadius
       width=0
       color {colorUseRule=UseColor_Specified, specifiedColor=<rgba 0 0 0 0>}
       cornerRadii=<0 0 0 0>
      }
      tintColor {colorUseRule=UseColor_Specified, specifiedColor=<rgba 1 1 1 1>}
      uVRegion {bIsValid=false, max=<0 0>, min=<0 0>}
     }
    }
   }
   slot {class=WidgetSwitcherSlot}
  }
  Text_SocialName_Hover {
   widgetClassPath=&FortSocialNameTextBlock
   bInherited=true
   bIsVariable=true
   parent=Switcher_SocialName
   widget=Text_SocialName_Hover
   properties {
    bOverrideDefaultStyle=true
    bTintInlineIcon=true
    mobileTextBlockScale=1.2
    platformIconDisplayRule=AlwaysIfDifferent
    scrollStyle=&TextScrollStyle-Base_C
    secondNameFontSize=14
    text="Nickname <PC color=\"82BDBDFF\" size=\"12\"></><text color=\"82BDBDFF\" size=\"12\">Display Name</>"
    defaultTextStyleOverride {
     overflowPolicy=Clip
     transformPolicy=None
     colorAndOpacity {colorUseRule=UseColor_Specified, specifiedColor=<rgba 0 0.004 0.063 1>}
     font {
      bForceMonospaced=false
      bMaterialIsStencil=false
      fontMaterial=null
      fontObject=&BurbankSmall
      letterSpacing=0
      monospacedWidth=1
      size=16
      skewAmount=0
      typefaceFontName=Black
      outlineSettings {
       bApplyOutlineToDropShadows=false
       bMiteredCorners=false
       bSeparateFillAlpha=false
       outlineMaterial=null
       outlineSize=0
       outlineColor=<rgba 0 0 0 1>
      }
     }
     highlightColor {colorUseRule=UseColor_Specified, specifiedColor=<rgba 0 0 0 1>}
     highlightShape {
      bIsDynamicallyLoaded=false
      drawAs=Image
      imageType=NoImage
      mirroring=NoMirror
      resourceName=None
      resourceObject=null
      tiling=NoTile
      imageSize=<32 32>
      margin=<ltrb 0 0 0 0>
      outlineSettings {
       bUseBrushTransparency=false
       roundingType=HalfHeightRadius
       width=0
       color {colorUseRule=UseColor_Specified, specifiedColor=<rgba 0 0 0 0>}
       cornerRadii=<0 0 0 0>
      }
      tintColor {colorUseRule=UseColor_Specified, specifiedColor=<rgba 1 1 1 1>}
      uVRegion {bIsValid=false, max=<0 0>, min=<0 0>}
     }
     selectedBackgroundColor {colorUseRule=UseColor_Foreground, specifiedColor=<rgba 1 0 1 1>}
     shadowColorAndOpacity=<rgba 0 0 0 1>
     shadowOffset=<0 0>
     strikeBrush {
      bIsDynamicallyLoaded=false
      drawAs=Image
      imageType=NoImage
      mirroring=NoMirror
      resourceName=None
      resourceObject=null
      tiling=NoTile
      imageSize=<32 32>
      margin=<ltrb 0 0 0 0>
      outlineSettings {
       bUseBrushTransparency=false
       roundingType=HalfHeightRadius
       width=0
       color {colorUseRule=UseColor_Specified, specifiedColor=<rgba 0 0 0 0>}
       cornerRadii=<0 0 0 0>
      }
      tintColor {colorUseRule=UseColor_Specified, specifiedColor=<rgba 1 1 1 1>}
      uVRegion {bIsValid=false, max=<0 0>, min=<0 0>}
     }
     underlineBrush {
      bIsDynamicallyLoaded=false
      drawAs=Image
      imageType=NoImage
      mirroring=NoMirror
      resourceName=None
      resourceObject=null
      tiling=NoTile
      imageSize=<32 32>
      margin=<ltrb 0 0 0 0>
      outlineSettings {
       bUseBrushTransparency=false
       roundingType=HalfHeightRadius
       width=0
       color {colorUseRule=UseColor_Specified, specifiedColor=<rgba 0 0 0 0>}
       cornerRadii=<0 0 0 0>
      }
      tintColor {colorUseRule=UseColor_Specified, specifiedColor=<rgba 1 1 1 1>}
      uVRegion {bIsValid=false, max=<0 0>, min=<0 0>}
     }
    }
    secondNameColor=<rgba 0.0039 0.1843 0.5961 1>
   }
   slot {class=WidgetSwitcherSlot}
  }
  Text_DisplayName {
   widgetClassPath=&CommonRichTextBlock
   bIsVariable=true
   parent=HBox_SocialName
   widget=Text_DisplayName
   properties {
    bOverrideDefaultStyle=true
    bTintInlineIcon=true
    mobileTextBlockScale=1.2
    renderOpacity=0.5
    scrollStyle=/FortUILibrary/Styles/Typography/Foundation/TextScrollStyle-NoFade.TextScrollStyle-NoFade_C
    text="User Name<PC/>"
    visibility=Collapsed
    defaultTextStyleOverride {
     overflowPolicy=Clip
     transformPolicy=None
     colorAndOpacity {colorUseRule=UseColor_Specified, specifiedColor=<rgba 1 1 1 1>}
     font {
      bForceMonospaced=false
      bMaterialIsStencil=false
      fontMaterial=null
      fontObject=&BurbankSmall
      letterSpacing=0
      monospacedWidth=1
      size=14
      skewAmount=0
      typefaceFontName=Black
      outlineSettings {
       bApplyOutlineToDropShadows=false
       bMiteredCorners=false
       bSeparateFillAlpha=false
       outlineMaterial=null
       outlineSize=0
       outlineColor=<rgba 0 0 0 1>
      }
     }
     highlightColor {colorUseRule=UseColor_Specified, specifiedColor=<rgba 0 0 0 1>}
     highlightShape {
      bIsDynamicallyLoaded=false
      drawAs=Image
      imageType=NoImage
      mirroring=NoMirror
      resourceName=None
      resourceObject=null
      tiling=NoTile
      imageSize=<32 32>
      margin=<ltrb 0 0 0 0>
      outlineSettings {
       bUseBrushTransparency=false
       roundingType=HalfHeightRadius
       width=0
       color {colorUseRule=UseColor_Specified, specifiedColor=<rgba 0 0 0 0>}
       cornerRadii=<0 0 0 0>
      }
      tintColor {colorUseRule=UseColor_Specified, specifiedColor=<rgba 1 1 1 1>}
      uVRegion {bIsValid=false, max=<0 0>, min=<0 0>}
     }
     selectedBackgroundColor {colorUseRule=UseColor_Foreground, specifiedColor=<rgba 1 0 1 1>}
     shadowColorAndOpacity=<rgba 0 0 0 1>
     shadowOffset=<0 0>
     strikeBrush {
      bIsDynamicallyLoaded=false
      drawAs=Image
      imageType=NoImage
      mirroring=NoMirror
      resourceName=None
      resourceObject=null
      tiling=NoTile
      imageSize=<32 32>
      margin=<ltrb 0 0 0 0>
      outlineSettings {
       bUseBrushTransparency=false
       roundingType=HalfHeightRadius
       width=0
       color {colorUseRule=UseColor_Specified, specifiedColor=<rgba 0 0 0 0>}
       cornerRadii=<0 0 0 0>
      }
      tintColor {colorUseRule=UseColor_Specified, specifiedColor=<rgba 1 1 1 1>}
      uVRegion {bIsValid=false, max=<0 0>, min=<0 0>}
     }
     underlineBrush {
      bIsDynamicallyLoaded=false
      drawAs=Image
      imageType=NoImage
      mirroring=NoMirror
      resourceName=None
      resourceObject=null
      tiling=NoTile
      imageSize=<32 32>
      margin=<ltrb 0 0 0 0>
      outlineSettings {
       bUseBrushTransparency=false
       roundingType=HalfHeightRadius
       width=0
       color {colorUseRule=UseColor_Specified, specifiedColor=<rgba 0 0 0 0>}
       cornerRadii=<0 0 0 0>
      }
      tintColor {colorUseRule=UseColor_Specified, specifiedColor=<rgba 1 1 1 1>}
      uVRegion {bIsValid=false, max=<0 0>, min=<0 0>}
     }
    }
   }
   slot {class=HorizontalBoxSlot, properties {padding=<ltrb 10 0 0 0>}}
  }
  JoinRequest {
   widgetClassPath=&CommonRichTextBlock
   bIsVariable=true
   parent=UserTextInfoVB
   widget=JoinRequest
   properties {
    bOverrideDefaultStyle=true
    bTintInlineIcon=true
    mobileTextBlockScale=1.2
    scrollStyle=&TextScrollStyle-NoFade_C
    text="<text color=\"F9FE49FF\">JOIN REQUEST!</>"
    visibility=Collapsed
    defaultTextStyleOverride {
     overflowPolicy=Clip
     transformPolicy=None
     colorAndOpacity {colorUseRule=UseColor_Specified, specifiedColor=<rgba 0.4452 1 1 1>}
     font {
      bForceMonospaced=false
      bMaterialIsStencil=false
      fontMaterial=null
      fontObject=&BurbankSmall
      letterSpacing=0
      monospacedWidth=1
      size=16
      skewAmount=0
      typefaceFontName=Black
      outlineSettings {
       bApplyOutlineToDropShadows=false
       bMiteredCorners=false
       bSeparateFillAlpha=false
       outlineMaterial=null
       outlineSize=0
       outlineColor=<rgba 0 0 0 1>
      }
     }
     highlightColor {colorUseRule=UseColor_Specified, specifiedColor=<rgba 0 0 0 1>}
     highlightShape {
      bIsDynamicallyLoaded=false
      drawAs=Image
      imageType=NoImage
      mirroring=NoMirror
      resourceName=None
      resourceObject=null
      tiling=NoTile
      imageSize=<32 32>
      margin=<ltrb 0 0 0 0>
      outlineSettings {
       bUseBrushTransparency=false
       roundingType=HalfHeightRadius
       width=0
       color {colorUseRule=UseColor_Specified, specifiedColor=<rgba 0 0 0 0>}
       cornerRadii=<0 0 0 0>
      }
      tintColor {colorUseRule=UseColor_Specified, specifiedColor=<rgba 1 1 1 1>}
      uVRegion {bIsValid=false, max=<0 0>, min=<0 0>}
     }
     selectedBackgroundColor {colorUseRule=UseColor_Foreground, specifiedColor=<rgba 1 0 1 1>}
     shadowColorAndOpacity=<rgba 0 0 0 1>
     shadowOffset=<0 0>
     strikeBrush {
      bIsDynamicallyLoaded=false
      drawAs=Image
      imageType=NoImage
      mirroring=NoMirror
      resourceName=None
      resourceObject=null
      tiling=NoTile
      imageSize=<32 32>
      margin=<ltrb 0 0 0 0>
      outlineSettings {
       bUseBrushTransparency=false
       roundingType=HalfHeightRadius
       width=0
       color {colorUseRule=UseColor_Specified, specifiedColor=<rgba 0 0 0 0>}
       cornerRadii=<0 0 0 0>
      }
      tintColor {colorUseRule=UseColor_Specified, specifiedColor=<rgba 1 1 1 1>}
      uVRegion {bIsValid=false, max=<0 0>, min=<0 0>}
     }
     underlineBrush {
      bIsDynamicallyLoaded=false
      drawAs=Image
      imageType=NoImage
      mirroring=NoMirror
      resourceName=None
      resourceObject=null
      tiling=NoTile
      imageSize=<32 32>
      margin=<ltrb 0 0 0 0>
      outlineSettings {
       bUseBrushTransparency=false
       roundingType=HalfHeightRadius
       width=0
       color {colorUseRule=UseColor_Specified, specifiedColor=<rgba 0 0 0 0>}
       cornerRadii=<0 0 0 0>
      }
      tintColor {colorUseRule=UseColor_Specified, specifiedColor=<rgba 1 1 1 1>}
      uVRegion {bIsValid=false, max=<0 0>, min=<0 0>}
     }
    }
   }
   slot {class=VerticalBoxSlot}
  }
  SizeBox_1 {
   widgetClassPath=&SizeBox
   parent=UserTextInfoVB
   widget=SizeBox_1
   properties {bOverride_MaxDesiredHeight=true, heightOverride=26, maxDesiredHeight=32}
   slot {class=VerticalBoxSlot, properties {verticalAlignment=VAlign_Center}}
  }
  Switcher_RichPresence {
   widgetClassPath=&CommonActivatableWidgetSwitcher
   bInherited=true
   bIsVariable=true
   parent=SizeBox_1
   widget=Switcher_RichPresence
   properties {transitionCurveType=Linear, transitionDuration=0.1}
   slot {class=SizeBoxSlot}
  }
  Text_RichPresence {
   widgetClassPath=&CommonRichTextBlock
   bInherited=true
   bIsVariable=true
   parent=Switcher_RichPresence
   widget=Text_RichPresence
   properties {
    bOverrideDefaultStyle=true
    mobileTextBlockScale=1.2
    scrollStyle=&TextScrollStyle-Base_C
    text=Last Interaction
    defaultTextStyleOverride {
     overflowPolicy=Clip
     transformPolicy=None
     colorAndOpacity {colorUseRule=UseColor_Specified, specifiedColor=<rgba 0.0331 0.402 1 1>}
     font {
      bForceMonospaced=false
      bMaterialIsStencil=false
      fontMaterial=null
      fontObject=&BurbankSmall
      letterSpacing=0
      monospacedWidth=1
      size=14
      skewAmount=0
      typefaceFontName=Bold
      outlineSettings {
       bApplyOutlineToDropShadows=false
       bMiteredCorners=false
       bSeparateFillAlpha=false
       outlineMaterial=null
       outlineSize=0
       outlineColor=<rgba 0 0 0 1>
      }
     }
     highlightColor {colorUseRule=UseColor_Specified, specifiedColor=<rgba 0 0 0 1>}
     highlightShape {
      bIsDynamicallyLoaded=false
      drawAs=Image
      imageType=NoImage
      mirroring=NoMirror
      resourceName=None
      resourceObject=null
      tiling=NoTile
      imageSize=<32 32>
      margin=<ltrb 0 0 0 0>
      outlineSettings {
       bUseBrushTransparency=false
       roundingType=HalfHeightRadius
       width=0
       color {colorUseRule=UseColor_Specified, specifiedColor=<rgba 0 0 0 0>}
       cornerRadii=<0 0 0 0>
      }
      tintColor {colorUseRule=UseColor_Specified, specifiedColor=<rgba 1 1 1 1>}
      uVRegion {bIsValid=false, max=<0 0>, min=<0 0>}
     }
     selectedBackgroundColor {colorUseRule=UseColor_Foreground, specifiedColor=<rgba 1 0 1 1>}
     shadowColorAndOpacity=<rgba 0 0 0 1>
     shadowOffset=<0 0>
     strikeBrush {
      bIsDynamicallyLoaded=false
      drawAs=Image
      imageType=NoImage
      mirroring=NoMirror
      resourceName=None
      resourceObject=null
      tiling=NoTile
      imageSize=<32 32>
      margin=<ltrb 0 0 0 0>
      outlineSettings {
       bUseBrushTransparency=false
       roundingType=HalfHeightRadius
       width=0
       color {colorUseRule=UseColor_Specified, specifiedColor=<rgba 0 0 0 0>}
       cornerRadii=<0 0 0 0>
      }
      tintColor {colorUseRule=UseColor_Specified, specifiedColor=<rgba 1 1 1 1>}
      uVRegion {bIsValid=false, max=<0 0>, min=<0 0>}
     }
     underlineBrush {
      bIsDynamicallyLoaded=false
      drawAs=Image
      imageType=NoImage
      mirroring=NoMirror
      resourceName=None
      resourceObject=null
      tiling=NoTile
      imageSize=<32 32>
      margin=<ltrb 0 0 0 0>
      outlineSettings {
       bUseBrushTransparency=false
       roundingType=HalfHeightRadius
       width=0
       color {colorUseRule=UseColor_Specified, specifiedColor=<rgba 0 0 0 0>}
       cornerRadii=<0 0 0 0>
      }
      tintColor {colorUseRule=UseColor_Specified, specifiedColor=<rgba 1 1 1 1>}
      uVRegion {bIsValid=false, max=<0 0>, min=<0 0>}
     }
    }
   }
   slot {class=WidgetSwitcherSlot}
  }
  Text_RichPresence_Hover {
   widgetClassPath=&CommonRichTextBlock
   bInherited=true
   bIsVariable=true
   parent=Switcher_RichPresence
   widget=Text_RichPresence_Hover
   properties {
    bOverrideDefaultStyle=true
    mobileTextBlockScale=1.2
    scrollStyle=&TextScrollStyle-Base_C
    text=Last Interaction
    defaultTextStyleOverride {
     overflowPolicy=Clip
     transformPolicy=None
     colorAndOpacity {colorUseRule=UseColor_Specified, specifiedColor=<rgba 0.0003 0.0284 0.314 1>}
     font {
      bForceMonospaced=false
      bMaterialIsStencil=false
      fontMaterial=null
      fontObject=&BurbankSmall
      letterSpacing=0
      monospacedWidth=1
      size=14
      skewAmount=0
      typefaceFontName=Bold
      outlineSettings {
       bApplyOutlineToDropShadows=false
       bMiteredCorners=false
       bSeparateFillAlpha=false
       outlineMaterial=null
       outlineSize=0
       outlineColor=<rgba 0 0 0 1>
      }
     }
     highlightColor {colorUseRule=UseColor_Specified, specifiedColor=<rgba 0 0 0 1>}
     highlightShape {
      bIsDynamicallyLoaded=false
      drawAs=Image
      imageType=NoImage
      mirroring=NoMirror
      resourceName=None
      resourceObject=null
      tiling=NoTile
      imageSize=<32 32>
      margin=<ltrb 0 0 0 0>
      outlineSettings {
       bUseBrushTransparency=false
       roundingType=HalfHeightRadius
       width=0
       color {colorUseRule=UseColor_Specified, specifiedColor=<rgba 0 0 0 0>}
       cornerRadii=<0 0 0 0>
      }
      tintColor {colorUseRule=UseColor_Specified, specifiedColor=<rgba 1 1 1 1>}
      uVRegion {bIsValid=false, max=<0 0>, min=<0 0>}
     }
     selectedBackgroundColor {colorUseRule=UseColor_Foreground, specifiedColor=<rgba 1 0 1 1>}
     shadowColorAndOpacity=<rgba 0 0 0 1>
     shadowOffset=<0 0>
     strikeBrush {
      bIsDynamicallyLoaded=false
      drawAs=Image
      imageType=NoImage
      mirroring=NoMirror
      resourceName=None
      resourceObject=null
      tiling=NoTile
      imageSize=<32 32>
      margin=<ltrb 0 0 0 0>
      outlineSettings {
       bUseBrushTransparency=false
       roundingType=HalfHeightRadius
       width=0
       color {colorUseRule=UseColor_Specified, specifiedColor=<rgba 0 0 0 0>}
       cornerRadii=<0 0 0 0>
      }
      tintColor {colorUseRule=UseColor_Specified, specifiedColor=<rgba 1 1 1 1>}
      uVRegion {bIsValid=false, max=<0 0>, min=<0 0>}
     }
     underlineBrush {
      bIsDynamicallyLoaded=false
      drawAs=Image
      imageType=NoImage
      mirroring=NoMirror
      resourceName=None
      resourceObject=null
      tiling=NoTile
      imageSize=<32 32>
      margin=<ltrb 0 0 0 0>
      outlineSettings {
       bUseBrushTransparency=false
       roundingType=HalfHeightRadius
       width=0
       color {colorUseRule=UseColor_Specified, specifiedColor=<rgba 0 0 0 0>}
       cornerRadii=<0 0 0 0>
      }
      tintColor {colorUseRule=UseColor_Specified, specifiedColor=<rgba 1 1 1 1>}
      uVRegion {bIsValid=false, max=<0 0>, min=<0 0>}
     }
    }
   }
   slot {class=WidgetSwitcherSlot}
  }
  Overlay_0 {
   widgetClassPath=&Overlay
   parent=HorizontalBox_2
   widget=Overlay_0
   slot {class=HorizontalBoxSlot, properties {horizontalAlignment=HAlign_Right}}
  }
  HorizontalBox_58 {
   widgetClassPath=&HorizontalBox
   parent=Overlay_0
   widget=HorizontalBox_58
   slot {class=OverlaySlot, properties {horizontalAlignment=HAlign_Fill, verticalAlignment=VAlign_Fill}}
  }
  PrivacySB {
   widgetClassPath=&SizeBox
   parent=HorizontalBox_58
   widget=PrivacySB
   properties {bOverride_HeightOverride=true, bOverride_WidthOverride=true, heightOverride=24, widthOverride=24}
   slot {class=HorizontalBoxSlot, properties {horizontalAlignment=HAlign_Right, verticalAlignment=VAlign_Center, padding=<ltrb 0 0 15 0>}}
  }
  Image_Lock {
   widgetClassPath=&Image
   bInherited=true
   bIsVariable=true
   parent=PrivacySB
   widget=Image_Lock
   properties {
    visibility=Collapsed
    brush {
     bIsDynamicallyLoaded=false
     drawAs=Image
     imageType=NoImage
     mirroring=NoMirror
     resourceName=None
     resourceObject=/Game/UI/Foundation/Textures/Icons/Locks/T-Icon-Lock-64
     tiling=NoTile
     imageSize=<24 24>
     margin=<ltrb 0 0 0 0>
     outlineSettings {
      bUseBrushTransparency=false
      roundingType=HalfHeightRadius
      width=0
      color {colorUseRule=UseColor_Specified, specifiedColor=<rgba 0 0 0 0>}
      cornerRadii=<0 0 0 0>
     }
     tintColor {colorUseRule=UseColor_Specified, specifiedColor=<rgba 0.0331 0.3968 1 1>}
     uVRegion {bIsValid=false, max=<0 0>, min=<0 0>}
    }
   }
   slot {class=SizeBoxSlot, properties {horizontalAlignment=HAlign_Center, verticalAlignment=VAlign_Center}}
  }
  Border_Highlight {
   widgetClassPath=&Border
   bInherited=true
   bIsVariable=true
   parent=HorizontalBox_58
   widget=Border_Highlight
   properties {
    horizontalAlignment=HAlign_Right
    visibility=SelfHitTestInvisible
    background {
     bIsDynamicallyLoaded=false
     drawAs=Image
     imageType=NoImage
     mirroring=NoMirror
     resourceName=None
     resourceObject=null
     tiling=NoTile
     imageSize=<64 64>
     margin=<ltrb 1 0 0 0>
     outlineSettings {
      bUseBrushTransparency=false
      roundingType=HalfHeightRadius
      width=0
      color {colorUseRule=UseColor_Specified, specifiedColor=<rgba 0 0 0 0>}
      cornerRadii=<0 0 0 0>
     }
     tintColor {colorUseRule=UseColor_Specified, specifiedColor=<rgba 1 1 1 0>}
     uVRegion {bIsValid=false, max=<0 0>, min=<0 0>}
    }
    padding=<ltrb 0 0 -32 0>
   }
   slot {class=HorizontalBoxSlot, properties {horizontalAlignment=HAlign_Right}}
  }
  InputActionWidget {
   widgetClassPath=&CommonActionWidget
   bInherited=true
   parent=Border_Highlight
   widget=InputActionWidget
   properties {renderOpacity=0, visibility=SelfHitTestInvisible, inputActions [{dataTable=&InputActionDataTable, rowName=Select}]}
   slot {class=BorderSlot, properties {horizontalAlignment=HAlign_Right, padding=<ltrb 0 0 -32 0>}}
  }
  EmptyButtonToEatOneMouseUpEvent {
   widgetClassPath=&IconTextButton
   bIsVariable=true
   parent=Overlay_1
   widget=EmptyButtonToEatOneMouseUpEvent
   properties {
    buttonText=""
    onButtonBaseClicked=(&self:WidgetTree.EmptyButtonToEatOneMouseUpEvent.HandleOnButtonClicked)
    onButtonBaseDoubleClicked=(&self:WidgetTree.EmptyButtonToEatOneMouseUpEvent.HandleOnButtonDoubleClicked)
    onButtonBaseHovered=(&self:WidgetTree.EmptyButtonToEatOneMouseUpEvent.HandleOnButtonHovered)
    onButtonBaseUnhovered=(&self:WidgetTree.EmptyButtonToEatOneMouseUpEvent.HandleOnButtonUnhovered)
    onSelectedChangedBase=(&self:WidgetTree.EmptyButtonToEatOneMouseUpEvent.HandleOnSelectedChanged)
    style=/FortUILibrary/Foundation/Buttons/ButtonStyles/Athena/ButtonStyle_TransparentList_Social.ButtonStyle_TransparentList_Social_C
    visibility=Collapsed
    hoveredSlateSoundOverride {resourceObject=null}
    pressedSlateSoundOverride {resourceObject=null}
   }
   slot {class=OverlaySlot, properties {horizontalAlignment=HAlign_Fill, verticalAlignment=VAlign_Fill}}
  }
  SelectBubble {
   widgetClassPath=&SizeBox
   parent=Overlay_1
   widget=SelectBubble
   properties {bOverride_WidthOverride=true, minDesiredWidth=96, visibility=Collapsed, widthOverride=96}
   slot {class=OverlaySlot, properties {horizontalAlignment=HAlign_Right, verticalAlignment=VAlign_Fill}}
  }
  AnimatedInputBorder_Select {
   widgetClassPath=&Border
   parent=SelectBubble
   widget=AnimatedInputBorder_Select
   properties {
    visibility=SelfHitTestInvisible
    background {
     bIsDynamicallyLoaded=false
     drawAs=Image
     imageType=NoImage
     mirroring=NoMirror
     resourceName=None
     resourceObject=/Game/UI/SocialPanel/Art/Materials/M_UI_UserHoverDark_Inst
     tiling=NoTile
     imageSize=<64 64>
     margin=<ltrb 1 0 0 0>
     outlineSettings {
      bUseBrushTransparency=false
      roundingType=HalfHeightRadius
      width=0
      color {colorUseRule=UseColor_Specified, specifiedColor=<rgba 0 0 0 0>}
      cornerRadii=<0 0 0 0>
     }
     tintColor {colorUseRule=UseColor_Specified, specifiedColor=<rgba 1 1 1 1>}
     uVRegion {bIsValid=false, max=<0 0>, min=<0 0>}
    }
    padding=<ltrb 0 0 0 0>
   }
   slot {class=SizeBoxSlot, properties {horizontalAlignment=HAlign_Right}}
  }
  HorizontalBox {
   widgetClassPath=&HorizontalBox
   parent=AnimatedInputBorder_Select
   widget=HorizontalBox
   slot {class=BorderSlot, properties {padding=<ltrb 0 0 0 0>}}
  }
  ArrowSpacingSB_Select {
   widgetClassPath=&SizeBox
   parent=HorizontalBox
   widget=ArrowSpacingSB_Select
   properties {bOverride_HeightOverride=true, bOverride_WidthOverride=true, heightOverride=72, widthOverride=72}
   slot {class=HorizontalBoxSlot, properties {horizontalAlignment=HAlign_Right, verticalAlignment=VAlign_Center}}
  }
  Arrow_Select {
   widgetClassPath=&Image
   parent=ArrowSpacingSB_Select
   widget=Arrow_Select
   properties {
    renderOpacity=0
    visibility=SelfHitTestInvisible
    brush {
     bIsDynamicallyLoaded=false
     drawAs=Image
     imageType=NoImage
     mirroring=NoMirror
     resourceName=None
     resourceObject=/Game/UI/SocialPanel/Art/Textures/T_UI_Icon_TreeArrow_64
     tiling=NoTile
     imageSize=<40 40>
     margin=<ltrb 0 0 0 0>
     outlineSettings {
      bUseBrushTransparency=false
      roundingType=HalfHeightRadius
      width=0
      color {colorUseRule=UseColor_Specified, specifiedColor=<rgba 0 0 0 0>}
      cornerRadii=<0 0 0 0>
     }
     tintColor {colorUseRule=UseColor_Specified, specifiedColor=<rgba 1 1 1 1>}
     uVRegion {bIsValid=false, max=<0 0>, min=<0 0>}
    }
    colorAndOpacity=<rgba 1 1 1 0.5>
   }
   slot {class=SizeBoxSlot, properties {horizontalAlignment=HAlign_Center, verticalAlignment=VAlign_Center}}
  }
  Spacer_3 {widgetClassPath=&Spacer, parent=HorizontalBox, widget=Spacer_3, properties {size=<24 24>}, slot {class=HorizontalBoxSlot}}
  UserActionMenuInputButton {
   widgetClassPath=/Game/UI/SocialPanel/UserActionMenuInputButton
   bIsVariable=true
   parent=HorizontalBox
   widget=UserActionMenuInputButton
   properties {
    bHideInputActionWithKeyboard=true
    onButtonBaseClicked=(&self:WidgetTree.UserActionMenuInputButton.HandleOnButtonClicked)
    onButtonBaseDoubleClicked=(&self:WidgetTree.UserActionMenuInputButton.HandleOnButtonDoubleClicked)
    onButtonBaseHovered=(&self:WidgetTree.UserActionMenuInputButton.HandleOnButtonHovered)
    onButtonBaseUnhovered=(&self:WidgetTree.UserActionMenuInputButton.HandleOnButtonUnhovered)
    onSelectedChangedBase=(&self:WidgetTree.UserActionMenuInputButton.HandleOnSelectedChanged)
    renderOpacity=0
    renderTransform {angle=0, scale=<1 1>, shear=<0 0>, translation=<-24 0>}
    triggeringInputAction {dataTable=&InputActionDataTable, rowName=GenericBack}
   }
   slot {class=HorizontalBoxSlot, properties {horizontalAlignment=HAlign_Right, verticalAlignment=VAlign_Center, padding=<ltrb -16 0 0 0>}}
  }
  SocialTagList {widgetClassPath=/Script/FortniteUI.SidebarSocialTagListBase, bInherited=true, widget=null}
  BuffetInfo {widgetClassPath=/Script/FortniteUI.FortSocialEventRSVPWidget, bInherited=true, widget=null}
  Image_TextMute {widgetClassPath=&Image, bInherited=true, widget=null}
  Image_EncourageEpicFriend {widgetClassPath=&Image, bInherited=true, widget=null}
  Text_LastInteraction {widgetClassPath=&CommonRichTextBlock, bInherited=true, widget=null}
  Text_LastInteraction_Hover {widgetClassPath=&CommonRichTextBlock, bInherited=true, widget=null}
 ]
}
